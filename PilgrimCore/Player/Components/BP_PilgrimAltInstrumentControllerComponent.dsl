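; Player controller component for Pilgrim quickplay that manages each player's instrument by track type, rarity and alt variant (keytar for guitar).
; Event router listeners on the server track song state, suppressing equip montages during songs, and check song metadata for keyboard starts on instrument changes.
; Swaps replace inventory item definitions resolved from a tier collection.
; Clients equip on pawn restart, and the server relays swap VFX through every player's component.
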(event EventBeginPlay
  ;[0.5782 1 0.5986] cache fort player controller, used in many places
  (bind _asfort_player_controller (Utilities|Casting|CastToFortPlayerController (Components|GetOwner)))
  (bind _output_get (Variables|Default|SetPlayerController _asfort_player_controller))
  (bind _eventrouter (Pilgrim|PilgrimWaitForEventRouter _output_get)
    (:OnEventRouterAvailable
      (Variables|Default|SetEvent%20Router _eventrouter)
      (Networking|SwitchHasAuthority
        (:Authority
          (Utilities|FlowControl|Sequence
            (:then_0
              ;[0.6036 0.9801 1] Server - While song is playing, disable weapon equip montage
              (bind (_asynctaskproxy _payload) (GameplayEventRouter|StartListeningToEvent (Variables|Default|GetEvent%20Router) "/Script/PilgrimQPSharedRuntime.PilgrimEvent_QuickplayState_BeganServer")
                (:then
                  (Utilities|Array|Add (Variables|Default|GetRouterEvents) _asynctaskproxy))
                (:OnEventReceived
                  (switch Utilities|FlowControl|Switch|SwitchonEPilgrimQuickplayState (Utilities|Struct|BreakPilgrimEventQuickplayStateBeganServer _payload)
                    (:Pregame
                      (Variables|Default|SetIsInGreenRoom true))
                    (:SongGameplay
                      (Variables|Default|SetCanPlayAltSwapVFX true)
                      (bind _asfort_player_controller_2 (Utilities|Casting|CastToFortPlayerController (Components|GetOwner)))
                      (Utilities|IsValid (Game|Player|GetPlayerPawn _asfort_player_controller_2)
                        (:Is%20Valid
                          (GameplayEffects|ApplyGameplayEffectToSelf :self (Ability|GetAbilitySystemComponent (Game|Player|GetPlayerPawn _asfort_player_controller_2)) :GameplayEffectClass (Variables|Default|GetDisableEquipAnimGE) :Level 0.0)
                          ;[0.7755 1 0.6711] The start of a song is also a reasonable time to reset our current rarity on server
                          (Variables|Default|SetCurrentRarity "Common"))))))))
            (:then_1
              (bind (_asynctaskproxy_2 _payload_2) (GameplayEventRouter|StartListeningToEvent (Variables|Default|GetEvent%20Router) "/Script/PilgrimQPSharedRuntime.PilgrimEvent_QuickplayState_EndedServer")
                (:then
                  (Utilities|Array|Add (Variables|Default|GetRouterEvents) _asynctaskproxy_2))
                (:OnEventReceived
                  ;;[0.6036 0.9801 1] We currently reenable right when the song ends.  If either this class or the powerups class switch back to default instrument at this moment or later, is the montage desirable or not?  For now, we're doing this, and we'll get feedback and adjust.  May need to make sure switching to default instrument happens before the state change, or else implement this a different way.
                  (switch Utilities|FlowControl|Switch|SwitchonEPilgrimQuickplayState (Utilities|Struct|BreakPilgrimEventQuickplayStateEndedServer _payload_2)
                    (:Pregame
                      (Variables|Default|SetIsInGreenRoom false))
                    (:SongGameplay
                      (Variables|Default|SetCanPlayAltSwapVFX false)
                      (bind _asfort_player_controller_3 (Utilities|Casting|CastToFortPlayerController (Components|GetOwner)))
                      (Utilities|IsValid (Game|Player|GetPlayerPawn _asfort_player_controller_3)
                        (:Is%20Valid
                          (GameplayEffects|RemoveActiveGameplayEffectbySourceEffect (Ability|GetAbilitySystemComponent (Game|Player|GetPlayerPawn _asfort_player_controller_3)) (Variables|Default|GetDisableEquipAnimGE) 0 -1))))))))
            (:then_2
              ;[1 0.6875 0.8795] Server - listen for spawn/respawn
              (bind _asfort_player_controller_4 (Utilities|Casting|CastToFortPlayerController (Components|GetOwner)))
              (Events|BindEventtoOnPawnSet (Variables|/Script/Engine.Controller|GetPlayerState _asfort_player_controller_4) (delegate Custom|On%20Pawn%20Set)))
            (:then_3
              ;;[0.1931 1 0.8498] Listen for Instrument change event
              (bind (_asynctaskproxy_3 _payload_3) (GameplayEventRouter|StartListeningToEvent (Variables|Default|GetEvent%20Router) "/Script/PilgrimCoreRuntime.PilgrimEvent_PlayerState_InstrumentChanged" (Variables|/Script/Engine.Controller|GetPlayerState (Variables|Default|GetPlayerController)))
                (:then
                  (Utilities|Array|Add (Variables|Default|GetRouterEvents) _asynctaskproxy_3))
                (:OnEventReceived
                  (if (== (Components|GetOwner) (Fortnite|GetFortPlayerControllerfromActor (Utilities|Struct|BreakPilgrimEventPlayerStateInstrumentChanged _payload_3)))
                    (Utilities|IsValid (Fortnite|GetFortPlayerControllerfromActor (Utilities|Struct|BreakPilgrimEventPlayerStateInstrumentChanged _payload_3))
                      (:Is%20Valid
                        (if (Utilities|Enum|NotEqual(Enum) (.NewInstrumentChoice (Utilities|Struct|BreakPilgrimEventPlayerStateInstrumentChanged _payload_3)) (Variables|Default|GetLastKnownTrackType))
                          ;;;[0.05673 0.52 1] Swapping to Alt instrument before trying to equip based off of the current/next song in playlist
                          (CallFunction|CheckforAltInstrument (.NewInstrumentChoice (Utilities|Struct|BreakPilgrimEventPlayerStateInstrumentChanged _payload_3)))
                          (Inventory|ExecuteInventoryItemDefinition :self (Fortnite|GetFortPlayerControllerfromActor (Utilities|Struct|BreakPilgrimEventPlayerStateInstrumentChanged _payload_3)) :ItemDefinition (CallFunction|GetAppropriateItemDefinition (.NewInstrumentChoice (Utilities|Struct|BreakPilgrimEventPlayerStateInstrumentChanged _payload_3)) (Variables|Default|GetCurrentRarity) (Variables|Default|GetIsCurrentlyAlt)) :Delay 0.0 :bForceExecute false :bActivateSlotAfterSettingFocused true))))))))))
        (:Remote
          (Utilities|FlowControl|Sequence
            (:then_0
              ;[0.9324 1 0.2077] Client - Checking if we should swap to alt instrument based on the first song in the setlist when a player readys up
              (bind _asynctaskproxy_4 (GameplayEventRouter|StartListeningToEvent (Variables|Default|GetEvent%20Router) "/Script/PilgrimCoreRuntime.PilgrimEvent_PlayerState_ReadinessChanged")
                (:then
                  (Utilities|Array|Add (Variables|Default|GetRouterEvents) _asynctaskproxy_4))
                (:OnEventReceived
                  _)))
            (:then_1
              ;;[1 0.3769 0.9147] Just incase players change song list while count down to game is happening. One last saftey check
              (bind _asynctaskproxy_5 (GameplayEventRouter|StartListeningToEvent (Variables|Default|GetEvent%20Router) "/Script/PilgrimCoreRuntime.PilgrimEvent_BandState_ReadinessChanged")
                (:then
                  (Utilities|Array|Add (Variables|Default|GetRouterEvents) _asynctaskproxy_5))
                (:OnEventReceived
                  _)))
            (:then_2
              ;[1 1 1] Client - Wait for pawn restart to equip instrument on spawn
              (Pawn|BindEventtoOnPossessedPawnChanged (Variables|Default|GetPlayerController) (delegate Custom|ClientPossessedPawnChanged))))
          (CallFunction|ServerCheckforAltInstrumentSwap))))))

(event Custom|ServerSetTrackType
  (params (NewTrackType :type EPilgrimTrackType%20Enum :default "TrackGuitar"))
  (Variables|Default|SetLastKnownTrackType NewTrackType))

(event EventEndPlay
  (params (EndPlayReason :type EEndPlayReason%20Enum :default "Destroyed"))
  (for _array_element (Variables|Default|GetRouterEvents)
    (GameplayEventRouter|StopListeningtoEvent _array_element))
  (Utilities|Array|Clear (Variables|Default|GetRouterEvents)))

(event Custom|On%20Pawn%20Set
  (params (Player :type Player%20State%20Object%20Reference) (NewPawn :type Pawn%20Object%20Reference) (OldPawn :type Pawn%20Object%20Reference))
  ;[1 0.6875 0.8795] Server - listen for spawn/respawn
  (Development|PrintString (Utilities|String|Append "AltInstrument: set pawn with " (Utilities|GetDisplayName NewPawn)))
  (Utilities|IsValid NewPawn
    (:Is%20Valid
      ;;[0.7755 1 0.6711] Player death resets rarity. Probably not needed because we've already reset it on DBNO, but just in case.
      (Variables|Default|SetCurrentRarity "Common")
      (Utilities|FlowControl|Sequence
        (:then_0
          ;[1 0.885 0.6811] Keep track of last known track type
          (bind _asfort_player_pawn (Utilities|Casting|CastToFortPlayerPawn NewPawn))
          (FortPawn|BindEventtoOnWeaponEquippedDelegate _asfort_player_pawn (delegate Custom|Equipped)))))))

(event Custom|Equipped
  (params (NewWeapon :type Fort%20Weapon%20Object%20Reference) (PrevWeapon :type Fort%20Weapon%20Object%20Reference))
  (bind _returnvalue (Actor|GetComponentbyClass NewWeapon "/SparksCosmetics/Components/BP_SparksInstrumentSettingsComponent.BP_SparksInstrumentSettingsComponent_C"))
  ;[1 0.885 0.6811] Keep track of last known track type
  (CallFunction|CheckPlaySwapVFX)
  (Utilities|IsValid NewWeapon
    (:Is%20Valid
      (Utilities|IsValid _returnvalue
        (:Is%20Valid
          (Variables|Default|SetLastKnownTrackType (Class|PilgrimCoreBPFL|SparksInstrumenttoPilgrimTrackType (Variables|/SparksCosmetics/Components/BP_SparksInstrumentSettingsComponent.BP_SparksInstrumentSettingsComponent_C|GetInstrumentType _returnvalue))))
        (:Is%20Not%20Valid
          (Variables|Default|SetLastKnownTrackType "TrackGuitar"))))))

(event Custom|SetInstrumentOnSpawn
  (bind _playercontroller (Variables|Default|GetPlayerController))
  (bind _returnvalue (Game|Player|GetPlayerPawn _playercontroller))
  (bind _lastknowntracktype (Variables|Default|GetLastKnownTrackType))
  (bind _item_definition (CallFunction|GetAppropriateItemDefinition _lastknowntracktype))
  ;[1 0.6875 0.8795] Set appropriate instrument on spawn or respawn
  ;;[1 0.8599 0] A guard check so we don't override a player selected instrument
  (Utilities|IsValid _returnvalue
    (:Is%20Valid
      (Utilities|IsValid (Weapon|GetCurrentWeapon _returnvalue)
        (:Is%20Valid
          ;;[1 0.4333 0.9431] Do we need to swap for non-standard instrument, or just select one of the standard ones we spawn with?
          (if (or (Variables|Default|GetIsCurrentlyAlt) (Utilities|Enum|NotEqual(Enum) (Variables|Default|GetCurrentRarity) "Common"))
            ;;[1 0.7013 0.3441] Swap from standard to variant
            (CallFunction|SwapfromWIDToWID _item_definition (CallFunction|GetAppropriateItemDefinition _lastknowntracktype (Variables|Default|GetCurrentRarity) (Variables|Default|GetIsCurrentlyAlt)))
            (else
              _)))
        (:Is%20Not%20Valid
          _))
      (Development|PrintString (Utilities|String|Append "AltInventory: set standard instrument  " (Utilities|String|EnumtoString _lastknowntracktype)))
      (Inventory|ExecuteInventoryItemDefinition :self (Variables|Default|GetPlayerController) :ItemDefinition _item_definition :Delay 0.0 :bForceExecute false :bActivateSlotAfterSettingFocused true))
    (:Is%20Not%20Valid
      (Development|PrintString "Player Pawn is not set, can't equip appropriate instrument!"))))

(event Custom|InitStartingInstrument
  (bind _initstartinginstrument_timerhandle (Variables|Default|GetInitStartingInstrument_TimerHandle))
  ;[1 0.3528 0.9657] Timer check to probe if we have any items owned first before equipping an instrument
  (if (Inventory|DoesInventoryOwnerHaveItems (Variables|Default|GetPlayerController) (Utilities|Array|MakeArray))
    (if (not (Utilities|Time|IsTimerPausedbyHandle _initstartinginstrument_timerhandle))
      (Utilities|Time|PauseTimerbyHandle _initstartinginstrument_timerhandle)
      (Fortnite|FortShippingLog self (Utilities|String|Append "[" (Utilities|GetObjectName (PlayerState|GetPawn (Variables|/Script/Engine.Controller|GetPlayerState (Variables|Default|GetPlayerController)))) "] Timer for InitStartingIntrument Has been paused")))
    (CallFunction|SetInstrumentonSpawn)))

(event Custom|ClientPossessedPawnChanged
  (params (OldPawn :type Pawn%20Object%20Reference) (NewPawn :type Pawn%20Object%20Reference))
  ;[1 1 1] Client - Wait for pawn restart to equip instrument on spawn
  (Utilities|FlowControl|Sequence
    (:then_0
      (Utilities|IsValid OldPawn
        (:Is%20Valid
          (Pawn|UnbindEventfromReceiveRestartedDelegate OldPawn (delegate Custom|ClientPawnRestarted)))))
    (:then_1
      (Utilities|IsValid NewPawn
        (:Is%20Valid
          (Pawn|BindEventtoReceiveRestartedDelegate NewPawn (delegate Custom|ClientPawnRestarted)))))))

(event Custom|ClientPawnRestarted
  (params (Pawn :type Pawn%20Object%20Reference))
  ;[1 1 1] Client - Wait for pawn restart to equip instrument on spawn
  (CallFunction|SetInstrumentonSpawn)
  (Pawn|UnbindEventfromReceiveRestartedDelegate Pawn (delegate Custom|ClientPawnRestarted)) ; Only set instrument on spawn once per pawn
)

(event Custom|Server%20Swap%20Alt%20Instrument
  (params (To%20Alt :type Boolean))
  ;[0.7247 0.4892 1] Server - change to alternate instrument, e.g. keytar for guitar.
  (Variables|Default|SetIsCurrentlyAlt To%20Alt)
  (CallFunction|ServerHandleSwapInstrument (Variables|Default|GetCurrentRarity) (Variables|Default|GetIsCurrentlyAlt)))

(event Custom|Server%20Swap%20Instrument%20Rarity
  (params (Rarity :type EFortRarity%20Enum :default "Common"))
  ;[0.5979 0.642 1] Server - Change Instrument Rarity, using whatever Alt flag we already have
  (bind _output_get (Variables|Default|SetCurrentRarity Rarity))
  (CallFunction|ServerHandleSwapInstrument _output_get (Variables|Default|GetIsCurrentlyAlt)))

(event Custom|Server%20Handle%20Swap%20Instrument
  (params (Rarity :type EFortRarity%20Enum :default "Common") (Use%20Alt%20Instrument :type Boolean))
  ;[0.789 1 0.4983] Server - Collect all the data we need to do the appropriate swap, and then do it.
  (bind _asfort_player_controller (Utilities|Casting|CastToFortPlayerController (Components|GetOwner)))
  (Utilities|IsValid (Class|FortPlayerPawn|GetLastWeaponEquippedonServer (Game|Player|GetPlayerPawn _asfort_player_controller))
    (:Is%20Valid
      (Utilities|IsValid (Actor|GetComponentbyClass (Class|FortPlayerPawn|GetLastWeaponEquippedonServer (Game|Player|GetPlayerPawn _asfort_player_controller)) "/SparksCosmetics/Components/BP_SparksInstrumentSettingsComponent.BP_SparksInstrumentSettingsComponent_C")
        (:Is%20Valid
          (CallFunction|ServerDoInstrumentSwap (Class|PilgrimCoreBPFL|SparksInstrumenttoPilgrimTrackType (Variables|/SparksCosmetics/Components/BP_SparksInstrumentSettingsComponent.BP_SparksInstrumentSettingsComponent_C|GetInstrumentType (Actor|GetComponentbyClass (Class|FortPlayerPawn|GetLastWeaponEquippedonServer (Game|Player|GetPlayerPawn _asfort_player_controller)) "/SparksCosmetics/Components/BP_SparksInstrumentSettingsComponent.BP_SparksInstrumentSettingsComponent_C"))) Rarity Use%20Alt%20Instrument))))))

(event Custom|Server%20Do%20Instrument%20Swap
  (params (Track%20Type :type EPilgrimTrackType%20Enum :default "TrackGuitar") (Rarity :type EFortRarity%20Enum :default "Common") (Use%20Alt%20Instrument :type Boolean))
  (bind _returnvalue (Utilities|String|ToString(Boolean) Use%20Alt%20Instrument))
  (bind _returnvalue_2 (not Use%20Alt%20Instrument))
  (bind _item_definition (CallFunction|GetAppropriateItemDefinition Track%20Type Rarity _returnvalue_2))
  (bind _item_definition_2 (CallFunction|GetAppropriateItemDefinition Track%20Type Rarity Use%20Alt%20Instrument))
  (bind _returnvalue_3 (Utilities|String|EnumtoString Rarity))
  (bind _returnvalue_4 (Utilities|String|EnumtoString Track%20Type))
  (bind _returnvalue_5 (Utilities|String|Append "AltInstrument: swap to track " _returnvalue_4 ", rarity " _returnvalue_3 ", alt " _returnvalue))
  ;[1 0.7013 0.3441] Swap from existing WID to new WID from params
  (Utilities|FlowControl|Sequence
    (:then_0
      (if (!= _item_definition _item_definition_2)
        (CallFunction|SwapfromWIDToWID _item_definition _item_definition_2)))
    (:then_1
      ;[0.1615 0.01402 0.0415] Some often-useful debug printing
      (Development|PrintString _returnvalue_5 true true "(R=0.890039,G=0.168231,B=1.000000,A=1.000000)")
      (Fortnite|FortShippingLog 0 _returnvalue_5))))

(event Custom|Swap%20From%20WID%20to%20WID
  (params (Remove%20Item :type Fort%20World%20Item%20Definition%20Object%20Reference) (Add%20Item :type Fort%20World%20Item%20Definition%20Object%20Reference))
  ;[1 0.7013 0.3441] Do an actual swap - remove one WID from inventory and add another
  (bind _asfort_player_controller (Utilities|Casting|CastToFortPlayerController (Components|GetOwner)))
  ;;[0.3283 1 0] Remove old item by WID
  (Inventory|RemoveItemfromInventoryOwner _asfort_player_controller Remove%20Item)
  ;;[1 0.5575 0.142] Give the player the item
  (Inventory|GiveItemtoInventoryOwner _asfort_player_controller Add%20Item))

(event Custom|Server_CheckForAltInstrumentSwap
  ;[1 0.8991 0.2581] Server - Checking if we should swap to alt instrument based of the starting song in playlist
  (CallFunction|CheckforAltInstrument (Variables|Default|GetLastKnownTrackType)))

(event Custom|Client_PlaySwapVFX
  (params (PlayerState :type Player%20State%20Object%20Reference))
  (bind _returnvalue (PlayerState|GetPawn PlayerState))
  ;[0.8254 1 0.4708] Client - Plays swapping to alt instrument VFX
  (Utilities|IsValid _returnvalue
    (:Is%20Valid
      (Niagara|SpawnSystemAttached "/SparksEffects/Effects/GameplayFX/InstrumentSwap/Particles/NS_Sparks_Instrument_Teleport_Out.NS_Sparks_Instrument_Teleport_Out" (Actor|GetComponentbyClass _returnvalue "/Script/Engine.SkeletalMeshComponent") ""))))

(fn CheckForAltInstrument
  (params (TrackType :type EPilgrimTrackType%20Enum :default "TrackGuitar"))
  (bind _songdata (LocalVariables|GetSongData))
  ;[0.1183 0.2471 1] Pilgrim Quickplay specific code!! grabs a different song meta data depending on the game state
  (if (Variables|Default|GetIsInGreenRoom)
    (bind _current_song (Default|GetCurrentSonginSetList))
    (LocalVariables|SetSongData _current_song)
    (else
      (bind _nextsong (Default|GetNextSonginSetList))
      (LocalVariables|SetSongData _nextsong)))
  ;[0.1437 0.2378 1] If the starting instrument isnt any of the base instrument, then it is an alt instrument
  (Utilities|IsValid _songdata
    (:Is%20Valid
      (switch Utilities|FlowControl|Switch|SwitchonESparksInstrumentType (Class|PilgrimSongMetadata|GetStartingInstrumentType _songdata TrackType)
        (:Guitar
          _)
        (:Bass
          _)
        (:Vocals
          _)
        (:Drum
          _)
        (:Keyboard
          (Variables|Default|SetIsCurrentlyAlt true)
          ;[0.1645 0.2462 1] Checking if we should swap out to the alt instrument  if it is not in the player inventory yet
          (bind _asfort_player_controller (Utilities|Casting|CastToFortPlayerController (Components|GetOwner)))
          (Utilities|IsValid (Inventory|FindExistingItemForDefinition :self _asfort_player_controller :ItemDefinition (CallFunction|GetAppropriateItemDefinition TrackType (Variables|Default|GetCurrentRarity) (Variables|Default|GetIsCurrentlyAlt)) :bInStorageVault false)
            (:Is%20Not%20Valid
              (CallFunction|ServerDoInstrumentSwap TrackType (Variables|Default|GetCurrentRarity) (Variables|Default|GetIsCurrentlyAlt)))))
        (:None
          _)))
    (:Is%20Not%20Valid
      _))
  (Variables|Default|SetIsCurrentlyAlt false))

(fn CheckPlaySwapVFX
  (Networking|SwitchHasAuthority
    (:Authority
      (if (and (or (Variables|Default|GetIsCurrentlyAlt) (Variables|Default|GetCached_IsCurrentlyAlt)) (Variables|Default|GetCanPlayAltSwapVFX))
        ;[1 1 1] Sending Instigating controller's playerstate to all client controllers so VFX plays on correct player pawn
        (for _array_element (Class|SparksMusicPlayspace|GetAllPlayerStates (Class|SparksMusicPlayspace|GetSparksMusicPlayspaceOfClass "/Script/PilgrimCoreRuntime.PilgrimPlayspace" (Variables|Default|GetPlayerController)))
          (Utilities|IsValid _array_element
            (:Is%20Valid
              (CallFunction|ClientPlaySwapVFX (Actor|GetComponentbyClass (PlayerState|GetPlayerController _array_element) "/PilgrimCore/Player/Components/BP_PilgrimAltInstrumentControllerComponent.BP_PilgrimAltInstrumentControllerComponent_C") (Variables|/Script/Engine.Controller|GetPlayerState (Variables|Default|GetPlayerController)))
              (Variables|Default|SetCached_IsCurrentlyAlt (Variables|Default|GetIsCurrentlyAlt)))))
        (else
          _)))))

(fn GetAppropriateItemDefinition
  (params (Track%20Type :type EPilgrimTrackType%20Enum :default "TrackGuitar") (Rarity :type EFortRarity%20Enum :default "Common") (Is%20Alt :type Boolean))
  (bind _instrumenttiercollection (Variables|Default|GetInstrumentTierCollection))
  (bind (_maininstruments_6_52de0b18466f271e8cb32991b19746e3 _keytar_10_2f917e32422c0534a7fbce81fef9f573) (Utilities|Struct|BreakPilgrimInstrumentTierCollection _instrumenttiercollection))
  (bind _use_tier_list (LocalVariables|GetUse%20Tier%20List))
  (bind _raritylist_9_3036c83e4ed4d251bb01a284f6f79710 (Utilities|Struct|BreakPilgrimInstrumentTierList _use_tier_list))
  (bind _rarity (Variables|GetRarity))
  (bind (_value _returnvalue) (Utilities|Map|Find _raritylist_9_3036c83e4ed4d251bb01a284f6f79710 _rarity))
  (bind (_value_2 _returnvalue_2) (Utilities|Map|Find _maininstruments_6_52de0b18466f271e8cb32991b19746e3 Track%20Type))
  (if Is%20Alt
    (if (or (Utilities|Enum|Equal(Enum) Track%20Type "TrackGuitar") (Utilities|Enum|Equal(Enum) Track%20Type "TrackPlasticGuitar"))
      (LocalVariables|SetUse%20Tier%20List _keytar_10_2f917e32422c0534a7fbce81fef9f573)
      (if _returnvalue
        (return _value)
        (else
          _))
      (else
        (if _returnvalue_2
          (LocalVariables|SetUse%20Tier%20List _value_2)
          (else
            _))))
    (Development|PrintString (Utilities|String|Append "AltInstrument : Couldn't find appropriate item definition for " (Utilities|String|EnumtoString (Variables|GetRarity)) " track " (Utilities|String|EnumtoString (Variables|GetTrack%20Type)) " and alt flag " (Utilities|String|ToString(Boolean) (Variables|GetIs%20Alt))))
    (return "/PilgrimCore/Weapons/WID_Pilgrim_BandQuickPlay_Guitar.WID_Pilgrim_BandQuickPlay_Guitar")
    (else
      _)))

(fn GetCurrentSongInSetList
  (bind _playercontroller (Variables|Default|GetPlayerController))
  (bind _returnvalue (Class|SparksMusicPlayspace|GetSparksMusicPlayspaceOfClass "/Script/PilgrimCoreRuntime.PilgrimPlayspace" _playercontroller))
  (bind _returnvalue_2 (Actor|GetComponentbyClass _returnvalue "/Script/PilgrimCoreRuntime.PilgrimCoreBandStateComponent"))
  (Utilities|IsValid _returnvalue_2
    (:Is%20Valid
      (return (Class|PilgrimCoreBandStateComponent|GetCurrentSong _returnvalue_2)))))

(fn GetNextSongInSetList
  (bind _playercontroller (Variables|Default|GetPlayerController))
  (bind _returnvalue (Class|SparksMusicPlayspace|GetSparksMusicPlayspaceOfClass "/Script/PilgrimCoreRuntime.PilgrimPlayspace" _playercontroller))
  (bind _returnvalue_2 (Actor|GetComponentbyClass _returnvalue "/Script/PilgrimCoreRuntime.PilgrimCoreBandStateComponent"))
  (Utilities|IsValid _returnvalue_2
    (:Is%20Valid
      (return (Class|PilgrimCoreBandStateComponent|GetNextSong _returnvalue_2)))))

(fn OnRep_IsCurrentlyAlt)

(fn SetInstrumentTierCollection
  (params (collection :type Pilgrim%20Instrument%20Tier%20Collection%20Structure))
  (Variables|Default|SetInstrumentTierCollection collection))

(fn SwapAltInstrument
  (params (To%20Alt :type Boolean))
  ;[0.6454 0.4708 1] don't bother if we're already there
  (if (!= To%20Alt (Variables|Default|GetIsCurrentlyAlt))
    (CallFunction|ServerSwapAltInstrument To%20Alt)
    (Development|PrintString "AltInstrument call to server happened" true true "(R=1.000000,G=0.925567,B=0.193532,A=1.000000)")))
