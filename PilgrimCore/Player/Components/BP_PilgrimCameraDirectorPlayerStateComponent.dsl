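; Pilgrim player state component that triggers a dynamic camera director event when its player emotes during gameplay.
; On authority, it watches the emote tag and, given a valid music battle instance, briefly waits, then stores the active montage length, clamped, defaulting to the min/max midpoint.
; Tag removal resets that length to -1 so the next emote retriggers.
; The length's RepNotify broadcasts the event on the pawn wherever cosmetics run.
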(event EventBeginPlay
  (bind _returnvalue (Components|GetOwner))
  (bind _returnvalue_2 (Class|SparksMusicPlayspace|GetSparksMusicPlayspaceOfClass "/Script/PilgrimCoreRuntime.PilgrimPlayspace" _returnvalue true))
  ;[0.5121 1 0.869] Wait for the event router (used later to broadcast an event)
  (bind _eventrouter (Pilgrim|PilgrimWaitForEventRouter (Components|GetOwner))
    (:OnEventRouterAvailable
      (Variables|Default|SetEventRouter _eventrouter)
      ;[0.5121 1 0.869] Only the client can play songs, so only set this up on the client
      (Networking|SwitchHasAuthority
        (:Authority
          ;[0.03434 1 0.227] Server
          ;;[0.5121 1 0.869] Wait to see when a player starts emoting, and start a dynamic camera event if we're in gameplay
          (Ability|Tasks|WaitGameplayTagAddToActor (Components|GetOwner) "(TagName=\"Gameplay.Action.Player.Emote\")"
            (:then
              ;;[0.5121 1 0.869] Wait for the emote tag to be removed from the player
              (Ability|Tasks|WaitGameplayTagRemoveFromActor (Components|GetOwner) "(TagName=\"Gameplay.Action.Player.Emote\")"
                (:Removed
                  ;;[0.5121 1 0.869] When the emote tag is removed, set the camera length on the server to -1. This ensures that we always play a dynamic camera when the player next starts an emote.
                  (Variables|Default|SetLengthOfCameraSequence -1.0))))
            (:Added
              ;;[0.5121 1 0.869] Check to see if there is a valid MBI before proceeding, so we don't try to trigger this outside of gameplay
              (Utilities|IsValid _returnvalue_2
                (:Is%20Valid
                  (Utilities|IsValid (Class|PilgrimPlayspace|GetMusicBattleInstanceforTeam _returnvalue_2 (Math|Conversions|ToInteger(Byte) (Team|GetActorTeam _returnvalue)))
                    (:Is%20Valid
                      ;;[0.5121 1 0.869] Give the emote a moment to spin up, and then get the length of its montage before firing the event.
                      (Utilities|FlowControl|Delay 0.1)
                      (bind _montagelengthsecs (CallFunction|GetLengthOfOwnerCurrentMontage))
                      (Variables|Default|SetLengthOfCameraSequence _montagelengthsecs))))))))))))

(fn GetLengthOfOwnerCurrentMontage
  (bind _returnvalue (Components|GetOwner))
  (bind _returnvalue_2 (Fortnite|GetFortPlayerControllerfromActor _returnvalue))
  (bind _returnvalue_3 (Game|Player|GetPlayerPawn _returnvalue_2))
  (bind _mesh (Variables|/Script/Engine.Character|GetMesh _returnvalue_3))
  (bind _returnvalue_4 (Components|SkeletalMesh|GetAnimInstance _mesh))
  (bind _returnvalue_5 (Animation|Montage|GetCurrentActiveMontage _returnvalue_4))
  ;[0.5121 1 0.869] By default, set the length to return to the average of our min and max length
  (LocalVariables|SetLengthToReturn (/ (+ (Variables|Default|GetMinSequenceLength) (Variables|Default|GetMaxSequenceLength)) 2.0))
  (Utilities|FlowControl|Sequence
    (:then_0
      ;[0.5121 1 0.869] Attempt to get the current active montage length, clamped between our min and max
      (Utilities|IsValid _returnvalue_2
        (:Is%20Valid
          (Utilities|IsValid _returnvalue_3
            (:Is%20Valid
              (Utilities|IsValid _returnvalue_4
                (:Is%20Valid
                  (Utilities|IsValid _returnvalue_5
                    (:Is%20Valid
                      (LocalVariables|SetLengthToReturn (Math|Float|Clamp(Float) (Animation|GetPlayLength _returnvalue_5) (Variables|Default|GetMinSequenceLength) (Variables|Default|GetMaxSequenceLength))))))))))))
    (:then_1
      ;[0.5121 1 0.869] Return the final length
      (return (LocalVariables|GetLengthToReturn)))))

(fn OnRep_LengthOfCameraSequence
  (bind _returnvalue (Components|GetOwner))
  (bind _returnvalue_2 (Class|SparksMusicPlayspace|GetSparksMusicPlayspaceOfClass "/Script/PilgrimCoreRuntime.PilgrimPlayspace" _returnvalue true))
  (bind _lengthofcamerasequence (Variables|Default|GetLengthOfCameraSequence))
  ;[1 0.7569 0] Don't wait for transition, start camera immediately if we have a length greater than 0
  (Networking|CanExecuteCosmeticEvents
    (:True
      (if (> _lengthofcamerasequence 0)
        (Utilities|IsValid _returnvalue_2
          (:Is%20Valid
            (bind _asfort_player_state (Utilities|Casting|CastToFortPlayerState (Components|GetOwner)))
            (Utilities|IsValid (PlayerState|GetCurrentPawn _asfort_player_state)
              (:Is%20Valid
                (Class|PilgrimPlayspace|BroadcastCameraDirectorDynamicEvent _returnvalue_2 (GameplayTags|MakeLiteralGameplayTagContainer "(GameplayTags=((TagName=\"CameraDirector.Pilgrim.Emote\"),(TagName=\"Gameplay.Action.Player.Emote\")))") (PlayerState|GetCurrentPawn _asfort_player_state) _lengthofcamerasequence false)))))))))
