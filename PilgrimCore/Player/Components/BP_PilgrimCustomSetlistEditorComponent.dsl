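; Player controller component that lets a player add and remove songs in the Pilgrim playspace's shared custom setlist.
; Local calls go through server RPCs to the playspace's PilgrimCustomSetlist, passing the owner's PlayerState.
; Removal by index reads the entry client-side, and the server removes it only if that player is the owner.
; It also forwards client analytics tracking and answers whether songs are already added by anyone.
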(event Custom|Server%20Add%20Song
  (params (Song%20To%20Add :type Name))
  (Fortnite|FortShippingLog self (Utilities|String|Append "[BP_PilgrimCustomSetlistEditorComponent] Server attempting to add song to setlist: " (Utilities|String|ToString(Name) Song%20To%20Add)))
  ;[1 1 1] Client->Server RPC to help route an AddSong call from local widgets to the server-side CustomSetlist
  (Class|PilgrimCustomSetlist|ServerAddSongforPlayer (CallFunction|GetCustomSetlisttoEdit) Song%20To%20Add (CallFunction|GetOwnerPlayerState)))

(event EventAddSongToSetlist
  (params (SongShortName :type Name :default "None"))
  (CallFunction|ServerAddSong SongShortName))

(event EventRemoveSongFromSetlist
  (params (SongShortName :type Name :default "None"))
  (CallFunction|ServerRemoveSong SongShortName))

(event Custom|Server%20Remove%20Song
  (params (Song%20To%20Remove :type Name))
  ;[1 1 1] Client->Server RPC to help route a RemoveSong call from local widgets to the server-side CustomSetlist
  (Class|PilgrimCustomSetlist|ServerRemoveSongforPlayer (CallFunction|GetCustomSetlisttoEdit) Song%20To%20Remove (CallFunction|GetOwnerPlayerState) false))

(event Custom|Client%20Remove%20Entry
  (params (Index :type Integer))
  (bind (_outsongshortname _outaddedbyplayer) (Class|PilgrimCustomSetlist|GetSetlistEntryatIndex (CallFunction|GetCustomSetlisttoEdit) Index))
  ;[1 1 1] Client-side helper to remove a song by index
  (CallFunction|ServerRemoveEntry Index _outsongshortname _outaddedbyplayer))

(event Custom|Server%20Remove%20Entry
  (params (Index :type Integer) (Song%20Short%20Name :type Name) (Player :type Player%20State%20Object%20Reference))
  ;[1 1 1] Client->Server RPC to help route a RemoveSongAtIndex call
  (if (== Player (CallFunction|GetOwnerPlayerState))
    (Class|PilgrimCustomSetlist|ServerRemoveSongatIndex (CallFunction|GetCustomSetlisttoEdit) Index Song%20Short%20Name Player)))

(event Custom|Add%20Song
  (params (SongToAdd :type Name))
  ;[1 1 1 0.5] Add and Remove Song Events
  (Class|PilgrimCustomSetlist|ClientTrackAddSongforPlayerforAnalytics (CallFunction|GetCustomSetlisttoEdit) SongToAdd (CallFunction|GetOwnerPlayerState))
  (CallFunction|ServerAddSong SongToAdd))

(event Custom|Remove%20Song
  (params (SongToRemove :type Name))
  ;[1 1 1 0.5] Add and Remove Song Events
  (Class|PilgrimCustomSetlist|ClientTrackRemoveSongforPlayerforAnalytics (CallFunction|GetCustomSetlisttoEdit) SongToRemove (CallFunction|GetOwnerPlayerState))
  (CallFunction|ServerRemoveSong SongToRemove))

(event Custom|Start%20Analytics%20Tracking
  ;[1 1 1 0.5] Start and Stop Analytics Tracking Events
  (Class|PilgrimCustomSetlist|ClientStartTrackingforAnalytics (CallFunction|GetCustomSetlisttoEdit) (CallFunction|GetOwnerPlayerState)))

(event Custom|Stop%20Analytics%20Tracking
  ;[1 1 1 0.5] Start and Stop Analytics Tracking Events
  (Class|PilgrimCustomSetlist|ClientStopTrackingforAnalytics (CallFunction|GetCustomSetlisttoEdit) (CallFunction|GetOwnerPlayerState)))

(fn AreSongsAddedToSetlist
  (params (SongShortNames :type Array%20of%20Names))
  (for _array_element SongShortNames
    (if (not (.ReturnValue (Class|PilgrimCustomSetlist|HasAnyoneAddedSong (CallFunction|GetCustomSetlisttoEdit) _array_element)))
      (return false)))
  (return true))

(fn GetCustomSetlistToEdit
  (return (Actor|GetComponentbyClass (Class|SparksMusicPlayspace|GetSparksMusicPlayspaceOfClass "/Script/PilgrimCoreRuntime.PilgrimPlayspace" (Components|GetOwner)) "/Script/PilgrimCoreRuntime.PilgrimCustomSetlist")))

(fn GetOwnerPlayerState
  (bind _asplayer_controller (Utilities|Casting|CastToPlayerController (Components|GetOwner)))
  (return (Variables|/Script/Engine.Controller|GetPlayerState _asplayer_controller)))

(fn IsSongAddedToSetlist
  (params (SongShortName :type Name :default "None"))
  (bind (_outaddedbyplayer _returnvalue) (Class|PilgrimCustomSetlist|HasAnyoneAddedSong (CallFunction|GetCustomSetlisttoEdit) SongShortName))
  (return _returnvalue))
