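; Player controller component that lets a player emote during their gem breaks (song breaks).
; On local controllers it listens through the event router for gem break start, end-pending, ended and song-stopped events.
; These drive server-side gameplay effects that grant or remove the allow-emote tag, toggle emote blocking and cancel emotes.
; Ham Sandwich song breaks swap the picker for quick-emote input, and remote clients force-close the picker when the tag is removed.
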(event EventBeginPlay
  (bind _returnvalue (Components|GetOwner))
  (bind _returnvalue_2 (Components|GetOwner))
  (bind _returnvalue_3 (Fortnite|GetFortPlayerControllerfromActor _returnvalue_2))
  (bind _asfort_player_controller (Utilities|Casting|CastToFortPlayerController _returnvalue))
  (bind _owningfortplayercontroller (Variables|Default|GetOwningFortPlayerController))
  ;[0.5121 1 0.869] When locally controlled, listen for gem break events.
  (Utilities|IsValid _asfort_player_controller
    (:Is%20Valid
      (Variables|Default|SetOwningFortPlayerController _asfort_player_controller)
      (if (Pawn|IsLocalController _asfort_player_controller)
        ;[0.5121 1 0.869] Wait for the event router
        (bind _eventrouter (Pilgrim|PilgrimWaitForEventRouter (Components|GetOwner))
          (:OnEventRouterAvailable
            (Variables|Default|SetEventRouter _eventrouter)
            (Utilities|FlowControl|Sequence
              (:then_0
                ;[0.5121 1 0.869] Listen for gem break start and stop events to enable and disable emoting. Note that the tag should be removed and the timer should be stopped before the end of the gem break, but make sure to try and remove the tag and stop the timer as a backup.
                ;; Gem Break Start: Allow emoting and maybe quick emoting.
                (bind (_asynctaskproxy _payload) (GameplayEventRouter|StartListeningToEvent (Variables|Default|GetEventRouter) "/Script/PilgrimCoreRuntime.PilgrimEvent_GemBreak_Started" (Variables|/Script/Engine.Controller|GetPlayerState (Fortnite|GetFortPlayerControllerfromActor (Components|GetOwner))))
                  (:OnEventReceived
                    ;;[0.3956 1 0.407 0.5] If song breaks are enable during ham sandwich, add the quick emoting mapping context when in Ham Sandwich and disable the normal picker
                    (if (and (Class|PilgrimCoreBPFL|IsHamSandwichSessionActive) (Class|PilgrimCoreBPFL|AreHamSandwichSongBreaksEnabled))
                      (Input|AddMappingContext (PlayerController|LocalPlayerSubsystems|GetEnhancedInputLocalPlayerSubsystem (Variables|Default|GetOwningFortPlayerController)) "/PilgrimCore/Player/Abilities/IMC_GreenRoomMappingContext_QuickEmoting.IMC_GreenRoomMappingContext_QuickEmoting" 1000 "(bIgnoreAllPressedKeysUntilRelease=True,bForceImmediately=False,bNotifyUserSettings=False)")
                      (UI|SetPickerEnabled (Variables|Default|GetOwningFortPlayerController) false)
                      (else
                        _))
                    (if (CallFunction|IsGemBreakforMe _payload)
                      (CallFunction|ApplyAllowEmoteDuringGemBreakTag (CallFunction|IsEmotingAllowedonClient))))))
              (:then_1
                ;; Gem Break End Pending : Remove emote ability.
                (bind (_asynctaskproxy_2 _payload_2) (GameplayEventRouter|StartListeningToEvent (Variables|Default|GetEventRouter) "/Script/PilgrimCoreRuntime.PilgrimEvent_GemBreak_End_Pending" (Variables|/Script/Engine.Controller|GetPlayerState (Fortnite|GetFortPlayerControllerfromActor (Components|GetOwner))))
                  (:OnEventReceived
                    (if (CallFunction|IsGemBreakforMe _payload_2)
                      (CallFunction|EndAllowEmoteDuringSongBreak)))))
              (:then_2
                ;; Gem Break Actually Ending: Double make sure we remove the emote ability.
                (bind (_asynctaskproxy_3 _payload_3) (GameplayEventRouter|StartListeningToEvent (Variables|Default|GetEventRouter) "/Script/PilgrimCoreRuntime.PilgrimEvent_GemBreak_Ended" (Variables|/Script/Engine.Controller|GetPlayerState (Fortnite|GetFortPlayerControllerfromActor (Components|GetOwner))))
                  (:OnEventReceived
                    (if (CallFunction|IsGemBreakforMe _payload_3)
                      (CallFunction|EndAllowEmoteDuringSongBreak)
                      ;;[0.3651 1 0.4225 0.5] Remove the quick emoting mapping context no matter what, and make sure the picker is enabled if we are in a ham sandwich song break
                      (Input|RemoveMappingContext (PlayerController|LocalPlayerSubsystems|GetEnhancedInputLocalPlayerSubsystem _owningfortplayercontroller) "/PilgrimCore/Player/Abilities/IMC_GreenRoomMappingContext_QuickEmoting.IMC_GreenRoomMappingContext_QuickEmoting" "(bIgnoreAllPressedKeysUntilRelease=True,bForceImmediately=False,bNotifyUserSettings=False)")
                      (if (and (Class|PilgrimCoreBPFL|IsHamSandwichSessionActive) (Class|PilgrimCoreBPFL|AreHamSandwichSongBreaksEnabled))
                        (UI|SetPickerEnabled _owningfortplayercontroller true))))))
              (:then_3
                ;;[1 0.9387 0.1441] If the song has stopped, stop the ability to emote
                (GameplayEventRouter|StartListeningToEvent (Variables|Default|GetEventRouter) "/Script/PilgrimCoreRuntime.PilgrimEvent_PilgrimGame_SongStopped"
                  (:OnEventReceived
                    (CallFunction|EndAllowEmoteDuringSongBreak))))
              (:then_4
                ;;[1 0.9387 0.1441] Listen for the tag that allows emoting (AllowEmoteDuringGemBreak) to be removed on the client (removed after changes on the the server propegate from removing the ability to emote) and then close any open emote wheel UI.
                (Networking|SwitchHasAuthority
                  (:Remote
                    (Ability|Tasks|WaitGameplayTagRemoveFromActor (Components|GetOwner) "(TagName=\"Pilgrim.Player.AllowEmoteDuringGemBreak\")"
                      (:Removed
                        (Utilities|IsValid _returnvalue_3
                          (:Is%20Valid
                            (UI|ForcePickerClose _returnvalue_3)))))))))))))))

(event Custom|ApplyAllowEmoteDuringGemBreakTag
  (params (EmotingAllowedOnClient :type Boolean))
  ;[0.2263 1 0.1274] Runs on server
  (Networking|SwitchHasAuthority
    (:Authority
      (bind _returnvalue (GameplayEffects|ApplyGameplayEffectToSelf :self (Ability|GetFortAbilitySystemComponent (Fortnite|GetFortPlayerControllerfromActor (Components|GetOwner))) :GameplayEffectClass (Variables|Default|GetGE_AllowEmote) :Level 0.0))
      (Variables|Default|SetAppliedGEHandle_AllowEmote _returnvalue)
      (if EmotingAllowedOnClient
        (CallFunction|EnableEmotingServer)))))

(event Custom|RemoveAllowEmoteDuringGemBreakTag
  (params (EmotingAllowedOnClient :type Boolean))
  ;[0.275 1 0.01217] Runs on server, clear the GE that's allowing them to emote.
  (Networking|SwitchHasAuthority
    (:Authority
      (GameplayEffects|RemoveActiveGameplayEffectbySourceEffect (Ability|GetFortAbilitySystemComponent (Fortnite|GetFortPlayerControllerfromActor (Components|GetOwner))) (Variables|Default|GetGE_AllowEmote) 0 -1)
      (if EmotingAllowedOnClient
        (CallFunction|DisableEmotingServer)))))

(event Custom|EndAllowEmoteDuringSongBreak
  ;[1 0.5312 0.9463] Stops the timer that checks for the end of the song break, and remove the tag that allows emoting
  (Utilities|Time|ClearandInvalidateTimerbyHandle (Variables|Default|GetGemBreakEndingTimer))
  (CallFunction|RemoveAllowEmoteDuringGemBreakTag (CallFunction|IsEmotingAllowedonClient)))

(event Custom|EnableEmotingServer
  (bind _returnvalue (Components|GetOwner))
  (bind _returnvalue_2 (Fortnite|GetFortPlayerControllerfromActor _returnvalue))
  ;[0.1661 1 0.7155] These events are run to enable or disable the ability for a player to emote
  (Utilities|SwitchHasAuthority
    (:Authority
      (Utilities|IsValid _returnvalue_2
        (:Is%20Valid
          (GameplayEffects|RemoveActiveEffectswithTags (Ability|GetFortAbilitySystemComponent _returnvalue_2) "(GameplayTags=((TagName=\"Pilgrim.Quickplay.GameplayBlockEmote\")))")
          (Class|PilgrimPlayerControllerComponent|DeactivateInputPreset (Actor|GetComponentbyClass _returnvalue_2 "/PilgrimCore/Player/Components/BP_PilgrimPlayerControllerComponent.BP_PilgrimPlayerControllerComponent_C")))))))

(event Custom|DisableEmotingServer
  (bind _returnvalue (Components|GetOwner))
  (bind _returnvalue_2 (Fortnite|GetFortPlayerControllerfromActor _returnvalue))
  (bind _returnvalue_3 (Ability|GetFortAbilitySystemComponent _returnvalue_2))
  ;[0.1661 1 0.7155] These events are run to enable or disable the ability for a player to emote
  (Utilities|SwitchHasAuthority
    (:Authority
      (Utilities|IsValid _returnvalue_2
        (:Is%20Valid
          (GameplayEffects|ApplyGameplayEffectToSelf :self _returnvalue_3 :GameplayEffectClass "/PilgrimCore/Player/Abilities/GE_PilgrimQuickplay_BlockEmoting.GE_PilgrimQuickplay_BlockEmoting_C" :Level 0.0)
          (Class|PilgrimPlayerControllerComponent|ActivateSelectedInputMappingContexts (Actor|GetComponentbyClass _returnvalue_2 "/PilgrimCore/Player/Components/BP_PilgrimPlayerControllerComponent.BP_PilgrimPlayerControllerComponent_C"))
          ;;[0.2889 0.3106 1] Stop any currently playing emotes
          (Tags|CancelAbilitiesWithTags _returnvalue_3 (GameplayTags|MakeLiteralGameplayTagContainer "(GameplayTags=((TagName=\"Gameplay.Action.Player.Emote\")))"))
          (Development|PrintString "AllowEmoteDuringGemBreakTag is removed from player, emoting is disabled" true true "(R=1.000000,G=0.662340,B=0.015694,A=1.000000)" 4.0))))))

(event Custom|CacheMBIAndPilgrimGame
  (bind _returnvalue (Components|GetOwner))
  (bind _returnvalue_2 (Fortnite|GetFortPlayerControllerfromActor _returnvalue))
  (bind _returnvalue_3 (Class|SparksMusicPlayspace|GetSparksMusicPlayspaceOfClass "/Script/PilgrimCoreRuntime.PilgrimPlayspace" _returnvalue))
  (bind _returnvalue_4 (Class|PilgrimPlayspace|GetMusicBattleInstanceforPlayer _returnvalue_3 _returnvalue_2))
  (bind _returnvalue_5 (Class|PilgrimPlayspace|TryGetPilgrimGameforPlayer _returnvalue_3 _returnvalue_2))
  ;[0.5121 1 0.869] Try to get a valid MBI and Pilgrim Game
  (Utilities|IsValid _returnvalue_3
    (:Is%20Valid
      (Utilities|IsValid _returnvalue_4
        (:Is%20Valid
          (Variables|Default|SetMBI _returnvalue_4)
          (Utilities|IsValid _returnvalue_2
            (:Is%20Valid
              (bind _asfort_player_state_athena (Utilities|Casting|CastToFortPlayerStateAthena (Variables|/Script/Engine.Controller|GetPlayerState _returnvalue_2)))
              (Variables|Default|SetOwningPlayerState _asfort_player_state_athena)
              (Utilities|IsValid _returnvalue_5
                (:Is%20Valid
                  (Variables|Default|SetPilgrimGame _returnvalue_5))
                (:Is%20Not%20Valid
                  (Utilities|FlowControl|DoOnce false
                    (:Completed
                      (Development|PrintString "NO VALID PILGRIM GAME" false false "(R=1.000000,G=0.093003,B=0.000000,A=1.000000)" 1e+01))))))
            (:Is%20Not%20Valid
              (Utilities|FlowControl|DoOnce false
                (:Completed
                  (Development|PrintString "No valid player controller" false false "(R=1.000000,G=0.093003,B=0.000000,A=1.000000)" 1e+01))))))
        (:Is%20Not%20Valid
          (Utilities|FlowControl|DoOnce false
            (:Completed
              (Development|PrintString "NO VALID MBI" false false "(R=1.000000,G=0.093003,B=0.000000,A=1.000000)" 1e+01))))))
    (:Is%20Not%20Valid
      (Utilities|FlowControl|DoOnce false
        (:Completed
          (Development|PrintString "No valid PilgrimPlayspace found" false false))))))

(fn IsEmotingAllowedOnClient
  (return (or (not (Class|PilgrimCoreBPFL|IsHamSandwichSessionActive)) (and (Class|PilgrimCoreBPFL|IsHamSandwichSessionActive) (Class|PilgrimCoreBPFL|AreHamSandwichSongBreaksEnabled)))))

(fn IsGemBreakForMe
  (params (PilgrimEvent_GemBreak_Base :type Pilgrim%20Event%20Gem%20Break%20Base%20Structure%20%28by%20ref%29))
  (bind (_break _listener _playerstate) (Utilities|Struct|BreakPilgrimEventGemBreakBase PilgrimEvent_GemBreak_Base))
  (bind (_id _bisvalidgembreak) (Utilities|Struct|BreakPilgrimGemBreak _break))
  (return (Math|Boolean|ANDBoolean (Utilities|IsValid (Variables|/Script/Engine.Controller|GetPlayerState (Fortnite|GetFortPlayerControllerfromActor (Components|GetOwner))) :!class /Script/Engine.KismetSystemLibrary) (== (Variables|/Script/Engine.Controller|GetPlayerState (Fortnite|GetFortPlayerControllerfromActor (Components|GetOwner))) _playerstate) _bisvalidgembreak)))
