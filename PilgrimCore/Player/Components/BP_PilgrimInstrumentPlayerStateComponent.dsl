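; Player state component that drives Pilgrim band-play instrument animation and cosmetic instrument type.
; On playspace entry it listens via the event router for quickplay state, song, clock and health events, and binds weapon equips.
; State, health trends and song end select the animatable state: none, without beat, animatable or playable.
; It also swaps alt instruments (CVar-gated reset to track default after songs) and toggles prop-suppression gameplay tags by state.
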(event Custom|On%20Manual%20Animation%20Start
  (params (Playspace :type Sparks%20Music%20Playspace%20Object%20Reference))
  (CallFunction|UpdateTrackType)
  (CallFunction|UpdateCosmeticInstrumentType)
  (CallFunction|UpdateIsCurrentlyAlt)
  (CallFunction|OnPilgrimStateChanged "SongGameplay")
  (CallFunction|OverridePlayspace Playspace)
  (CallFunction|LinkInstrumentAnimationLayer (Variables|Default|GetPlayerAnimations)))

(event Mojo|MusicPlayspace|EventEnteredSparksMusicPlayspace
  (params (Playspace :type Sparks%20Music%20Playspace%20Object%20Reference))
  (bind _returnvalue (CallFunction|GetOwnerPawn))
  (|CallParentFunction Playspace)
  (Networking|CanExecuteCosmeticEvents
    (:True
      (Utilities|FlowControl|Sequence
        (:then_0
          (bind _eventrouter (Pilgrim|PilgrimWaitForEventRouter (Components|GetOwner))
            (:OnEventRouterAvailable
              (Variables|Default|SetEventRouter _eventrouter)
              (bind _aspilgrim_playspace (Utilities|Casting|CastToPilgrimPlayspace (Variables|PawnComponent_PilgrimInstrumentAnim|GetSparksPlayspace)))
              (Variables|Default|SetPilgrimPlayspace _aspilgrim_playspace)
              (Utilities|FlowControl|Sequence
                (:then_0
                  ;[0.633 1 0.6611] track game pilgrim state
                  (bind (_asynctaskproxy _payload) (GameplayEventRouter|StartListeningToEvent (Variables|Default|GetEventRouter) "/Script/PilgrimQPSharedRuntime.PilgrimEvent_QuickplayState_BeganClient")
                    (:then
                      (Utilities|Array|Add (Variables|Default|GetGameplay%20Event%20Handles) _asynctaskproxy))
                    (:OnEventReceived
                      (CallFunction|OnPilgrimStateChanged (Utilities|Struct|BreakPilgrimEventQuickplayStateBeganClient _payload)))))
                (:then_1
                  (CallFunction|UpdateMusicClock)
                  (CallFunction|UpdateAnimatableState))
                (:then_2
                  ;[0.6314 0.7376 1] Get the PilgrimGame - needed so we can get the clock
                  (bind _asynctaskproxy_2 (GameplayEventRouter|StartListeningToEvent (Variables|Default|GetEventRouter) "/Script/PilgrimCoreRuntime.PilgrimEvent_PilgrimGame_PlayingSong")
                    (:then
                      (Utilities|Array|Add (Variables|Default|GetGameplay%20Event%20Handles) _asynctaskproxy_2))
                    (:OnEventReceived
                      (CallFunction|UpdateMusicClock))))
                (:then_3
                  ;[1 0.4475 0.4674] Listen for new battle data events
                  (bind (_asynctaskproxy_3 _payload_2) (GameplayEventRouter|StartListeningToEvent (Variables|Default|GetEventRouter) "/Script/PilgrimCoreRuntime.PilgrimEvent_BandState_PlayerHealthChanged" (CallFunction|GetOwnerPawn))
                    (:then
                      (Utilities|Array|Add (Variables|Default|GetGameplay%20Event%20Handles) _asynctaskproxy_3)
                      ;[1 0.8725 0.3209] Make sure we use the correct instrument when we get a new song
                      (bind (_asynctaskproxy_4 _payload_3) (GameplayEventRouter|StartListeningToEvent (Variables|Default|GetEventRouter) "/Script/PilgrimCoreRuntime.PilgrimEvent_MBI_ClientGotSongToPlay")
                        (:then
                          (Utilities|Array|Add (Variables|Default|GetGameplay%20Event%20Handles) _asynctaskproxy_4))
                        (:OnEventReceived
                          (CallFunction|OnClientGotSongtoPlay _payload_3))))
                    (:OnEventReceived
                      (CallFunction|OnPlayerHealthChanged _payload_2))))))))
        (:then_1
          ;[1 0.3514 0.6742] Pay attention to spawns/respawns, and weapon selection
          (Utilities|FlowControl|Sequence
            (:then_0
              (CallFunction|UpdateTrackType)
              (CallFunction|UpdateCosmeticInstrumentType)
              (CallFunction|UpdateIsCurrentlyAlt))
            (:then_1
              (Utilities|IsValid _returnvalue
                (:Is%20Valid
                  (bind _asfort_player_pawn (Utilities|Casting|CastToFortPlayerPawn _returnvalue))
                  (FortPawn|BindEventtoOnWeaponEquippedDelegate _asfort_player_pawn (delegate Custom|Weapon%20Equipped)))))))))))

(event Custom|OnGotPilgrimGame_Event
  (params (PilgrimGame :type Pilgrim%20Game%20Object%20Reference))
  (if (!= PilgrimGame (Variables|Default|GetPilgrimGame))
    (Variables|Default|SetPilgrimGame PilgrimGame)
    (CallFunction|UpdateMusicClock)))

(event Custom|BindMidiEventListeners
  ;[0 0.1203 1] Listen for the music ending event
  (bind _asynctaskproxy (MidiEvents|ListenForMidiTextEvent (Variables|SparksInstrumentAnimPawnComponent|GetMidiEventDriver) (Utilities|String|StringToName (Class|PilgrimCoreBPFL|PilgrimTrackTypetoMidiTrackName "TrackEvents")) (MidiEvents|TextEvents|Events|GetEventsMidiTextStringfromEnum "NewEnumerator4") 0.0 "NoPreviousEvents")
    (:then
      (Utilities|Array|Add (Variables|Default|GetMidi%20Event%20Listener%20Handles) _asynctaskproxy))
    (:OnMidiEventStart
      (CallFunction|OnSongEnded))))

(event Custom|Weapon%20Equipped
  (params (NewWeapon :type Fort%20Weapon%20Object%20Reference) (PrevWeapon :type Fort%20Weapon%20Object%20Reference))
  ;[1 0.3514 0.6742] Pay attention to spawns/respawns, and weapon selection
  (CallFunction|UpdateTrackType)
  (CallFunction|UpdateCosmeticInstrumentType)
  (CallFunction|UpdateIsCurrentlyAlt))

(event Custom|OnPlayerHealthChanged
  (params (PlayerHealthEvent :type Pilgrim%20Event%20Band%20State%20Player%20Health%20Changed%20Structure%20%28by%20ref%29))
  (bind (_lastplayerhealth _newplayerhealth) (Utilities|Struct|BreakPilgrimEventBandStatePlayerHealthChanged PlayerHealthEvent))
  ;[3 0.33 0.33] Health Increase / Decrease Events
  (if (< _lastplayerhealth _newplayerhealth)
    (Utilities|FlowControl|Sequence
      (:then_0
        (goto _label Close))
      (:then_1
        ;;[1 1 1] Call the animatable state changed if this value has changed
        (if (== (Variables|Default|GetGemFailThresholdReached) true)
          (Variables|Default|SetGemFailThresholdReached false)
          (CallFunction|UpdateAnimatableState)))
      (:then_2
        (Utilities|FlowControl|DoOnce false :!label _label_2
          (:Completed
            (Utilities|FlowControl|RetriggerableDelay 1.0)))))
    (else
      (Utilities|FlowControl|Sequence
        (:then_0
          (goto _label Open))
        (:then_1
          (goto _label_2)))))
  (Utilities|FlowControl|Gate true :!label _label
    (:Exit
      (if (== (Variables|Default|GetGemFailThresholdReached) false)
        (Variables|Default|SetGemFailThresholdReached true)
        (CallFunction|UpdateAnimatableState)))))

(event Custom|OnClientGotSongToPlay
  (params (PilgrimEvent_MBI_ClientGotSongToPlay :type Pilgrim%20Event%20MBI%20Client%20Got%20Song%20To%20Play%20Structure%20%28by%20ref%29))
  (bind _songtoplay (Utilities|Struct|BreakPilgrimEventMBIClientGotSongtoPlay PilgrimEvent_MBI_ClientGotSongToPlay))
  (bind _tracktype (Variables|Default|GetTrackType))
  (bind _returnvalue (Class|PilgrimSongMetadata|GetStartingInstrumentType _songtoplay _tracktype))
  ;[1 0.9074 0.3476] Swap instrument if needed
  (Utilities|IsValid _songtoplay
    (:Is%20Valid
      (if (Utilities|Enum|NotEqual(Enum) _returnvalue (Variables|SparksInstrumentAnimPawnComponent|GetCosmeticInstrumentType))
        (CallFunction|SwapCosmeticInstrument _returnvalue)))))

(fn CanSwapToInstrument
  (params (NewInstrumentType :type ESparksInstrumentType%20Enum :default "Guitar"))
  (if (> (Fortnite|GetNumericCVar (Variables|Default|GetCanSwapInstrumentsCVarName)) 0)
    ;[1 0.5 0.5] All instrument swaps are allowed
    (return true)
    (else
      ;[1 0.5 0.5] Only guitar <-> keytar is allowed
      (return (Utilities|Select :Index (Variables|Default|GetTrackType) :TrackGuitar (or (Utilities|Enum|Equal(Enum) NewInstrumentType "Guitar") (Utilities|Enum|Equal(Enum) NewInstrumentType "Keyboard")) :TrackBass false :TrackVocals false :TrackDrum false :TrackPlasticGuitar (or (Utilities|Enum|Equal(Enum) NewInstrumentType "Guitar") (Utilities|Enum|Equal(Enum) NewInstrumentType "Keyboard")) :TrackPlasticBass false :TrackPlasticDrum false :TrackProVocals false :TrackProCymbals false :TrackEvents false :TrackNone false)))))

(fn GetPilgrimState
  (return (Variables|Default|GetPilgrimState)))

(fn GetPlayerAnimations
  (return (Variables|Default|GetPlayerAnimations)))

(fn GetTrackInstrumentType
  (return (Utilities|Select :Index (Variables|Default|GetTrackType) :TrackGuitar "Guitar" :TrackBass "Bass" :TrackVocals "Vocals" :TrackDrum "Drum" :TrackPlasticGuitar "Guitar" :TrackPlasticBass "Bass" :TrackPlasticDrum "Drum" :TrackProVocals "Vocals" :TrackProCymbals "Drum" :TrackEvents "None" :TrackNone "None")))

(fn InBandPlayGameState
  (return (Math|Boolean|ANDBoolean (Utilities|Enum|NotEqual(Enum) (Variables|Default|GetPilgrimState) "Pregame") (Utilities|Enum|NotEqual(Enum) (Variables|Default|GetPilgrimState) "Loading") (Utilities|Enum|NotEqual(Enum) (Variables|Default|GetPilgrimState) "Invalid"))))

(fn IsUsable
  (return (and (Utilities|IsValid (Variables|Default|GetPilgrimPlayspace) :!class /Script/Engine.KismetSystemLibrary) (Utilities|Select :Index (CallFunction|GetPilgrimState) :Loading false :Pregame false :Preintro false :Intro true :SongGameplay true :Outro true :SongResults true :SetResults true))))

(fn OnPilgrimStateChanged
  (params (PilgrimState :type EPilgrimQuickplayState%20Enum :default "Invalid"))
  (bind _returnvalue (CallFunction|GetTrackInstrumentType))
  (Utilities|FlowControl|Sequence
    (:then_0
      (CallFunction|UpdateIsCurrentlyAlt))
    (:then_1
      (Variables|Default|SetPilgrimState PilgrimState)
      (CallFunction|UpdateSequencePropTags)
      (Variables|Default|SetGotSongEndedEvent false))
    (:then_2
      ;[0.7054 1 0.9601] Swap back to the default instrument for our track type so that we don't confuse other systems about which track we have selected
      (switch Utilities|FlowControl|Switch|SwitchonEPilgrimQuickplayState PilgrimState
        (:Pregame
          _)
        (:Outro
          _)
        (:SongResults
          _)
        (:SetResults
          _))
      (if (Utilities|Enum|NotEqual(Enum) (CallFunction|GetTrackInstrumentType) (Variables|SparksInstrumentAnimPawnComponent|GetCosmeticInstrumentType))
        (if (> (Fortnite|GetNumericCVar "Pilgrim.EquipDefaultInstrumentInResults") 0)
          ;;[1 1 1] Guitar does not exhibit any issues showing keytar
          (if (and (Utilities|Enum|Equal(Enum) (Variables|Default|GetTrackType) "TrackGuitar") (or (Utilities|Enum|Equal(Enum) (Variables|SparksInstrumentAnimPawnComponent|GetCosmeticInstrumentType) "Keyboard") (Utilities|Enum|Equal(Enum) (Variables|SparksInstrumentAnimPawnComponent|GetCosmeticInstrumentType) "Guitar")))
            _
            (elif (Utilities|Enum|NotEqual(Enum) (Variables|SparksInstrumentAnimPawnComponent|GetCosmeticInstrumentType) _returnvalue)
              (CallFunction|SwapCosmeticInstrument _returnvalue))))))
    (:then_3
      ;[1 0.7999 0.9288] Check if Animatable state has changed
      (CallFunction|UpdateAnimatableState))))

(fn OnSongEnded
  (Utilities|FlowControl|Sequence
    (:then_0
      (Variables|Default|SetGotSongEndedEvent true))
    (:then_1
      ;[1 0.7999 0.9288] Check if Animatable state has changed
      (CallFunction|UpdateAnimatableState))))

(fn SwapCosmeticInstrument
  (params (NewInstrumentType :type ESparksInstrumentType%20Enum :default "Guitar"))
  (bind _returnvalue (Components|GetOwner))
  (bind _returnvalue_2 (Fortnite|GetFortPlayerControllerfromActor _returnvalue))
  (bind _returnvalue_3 (Actor|GetComponentbyClass _returnvalue_2 "/PilgrimCore/Player/Components/BP_PilgrimAltInstrumentControllerComponent.BP_PilgrimAltInstrumentControllerComponent_C"))
  (bind _output_get (Variables|Default|SetIsCurrentlyAlt (Utilities|Enum|NotEqual(Enum) NewInstrumentType (CallFunction|GetTrackInstrumentType))))
  (Utilities|FlowControl|Sequence
    (:then_0
      (Utilities|IsValid _returnvalue_2
        (:Is%20Valid
          (Utilities|IsValid _returnvalue_3
            (:Is%20Valid
              (Class|BPPilgrimAltInstrumentControllerComponent|SwapAltInstrument _returnvalue_3 _output_get))))))
    (:then_1
      (bind _asplayer_state (Utilities|Casting|CastToPlayerState _returnvalue))
      (Utilities|IsValid (Actor|GetComponentbyClass (PlayerState|GetPawn _asplayer_state) "/PilgrimCore/Player/Components/BP_PilgrimAltInstrumentPawnComponent.BP_PilgrimAltInstrumentPawnComponent_C")
        (:Is%20Valid
          (Class|BPPilgrimAltInstrumentPawnComponent|SwapAltInstrument (Actor|GetComponentbyClass (PlayerState|GetPawn _asplayer_state) "/PilgrimCore/Player/Components/BP_PilgrimAltInstrumentPawnComponent.BP_PilgrimAltInstrumentPawnComponent_C") _output_get))))))

(fn UpdateAnimatableState
  (Utilities|IsValid (Variables|Default|GetPilgrimPlayspace)
    (:Is%20Valid
      (switch Utilities|FlowControl|Switch|SwitchonEPilgrimQuickplayState (CallFunction|GetPilgrimState)
        (:Loading
          _)
        (:Pregame
          _)
        (:Preintro
          _)
        (:Intro
          _)
        (:SongGameplay
          (if (Variables|Default|GetGotSongEndedEvent)
            ;[1 0.5 0.5] Animation should run, just without a beat
            (Variables|SparksInstrumentAnimPawnComponent|SetAnimatableState "AnimatableWithoutBeat" :!label _label)
            (elif (Variables|Default|GetGemFailThresholdReached)
              ;[1 0.5 0.5] Animation should run, just not playing the instrument
              (Variables|SparksInstrumentAnimPawnComponent|SetAnimatableState "Animatable")
              (else
                ;[1 0.5 0.5] Animation should be playing the instrument
                (Variables|SparksInstrumentAnimPawnComponent|SetAnimatableState "AnimatableAndPlayable")))))
        (:Outro
          (goto _label))
        (:SongResults
          _)
        (:SetResults
          _)))
    (:Is%20Not%20Valid
      _))
  ;[1 0.5 0.5] Animation should not be running
  (Variables|SparksInstrumentAnimPawnComponent|SetAnimatableState "None"))

(fn UpdateCosmeticInstrumentType
  (bind _asfort_pawn (Utilities|Casting|CastToFortPawn (CallFunction|GetOwnerPawn)))
  (Utilities|IsValid (Weapon|GetCurrentWeapon _asfort_pawn)
    (:Is%20Valid
      (Utilities|IsValid (Actor|GetComponentbyClass (Weapon|GetCurrentWeapon _asfort_pawn) "/Script/SparksCoreCosmeticsRuntime.SparksInstrumentComponentBase")
        (:Is%20Valid
          (Variables|SparksInstrumentAnimPawnComponent|SetCosmeticInstrumentType (Class|SparksInstrumentComponentBase|GetInstrumentType (Actor|GetComponentbyClass (Weapon|GetCurrentWeapon _asfort_pawn) "/Script/SparksCoreCosmeticsRuntime.SparksInstrumentComponentBase"))))))))

(fn UpdateIsCurrentlyAlt
  (bind _returnvalue (CallFunction|GetTrackInstrumentType))
  (bind _returnvalue_2 (Utilities|Enum|NotEqual(Enum) (Variables|SparksInstrumentAnimPawnComponent|GetCosmeticInstrumentType) _returnvalue))
  (bind _cosmeticinstrumenttype (Variables|SparksInstrumentAnimPawnComponent|GetCosmeticInstrumentType))
  ;[1 1 1] Change the IsCurrentlyAlt flag if
  ; 1. We have an instrument type
  ; 2. The IsCurrentlyAlt flag is not correct
  (if (and (< _cosmeticinstrumenttype (- (Utilities|Enum|GetnumberofentriesinESparksInstrumentType) 2)) (< _returnvalue (- (Utilities|Enum|GetnumberofentriesinESparksInstrumentType) 2)))
    (if (!= (Variables|Default|GetIsCurrentlyAlt) _returnvalue_2)
      (Variables|Default|SetIsCurrentlyAlt _returnvalue_2))))

(fn UpdateMusicClock
  (Variables|SparksInstrumentAnimPawnComponent|SetMusicClock (Class|PilgrimPlayspace|GetMusicClock 0)))

(fn UpdateSequencePropTags
  (bind _returnvalue (Ability|GetAbilitySystemComponent (CallFunction|GetOwnerPlayerState)))
  (bind _returnvalue_2 (GameplayTags|MakeLiteralGameplayTag "(TagName=\"SparksAnimations.ShouldNotSpawnProp.PilgrimInstruments\")"))
  (bind _returnvalue_3 (GameplayTags|MakeLiteralGameplayTag "(TagName=\"SparksAnimations.ShouldNotSpawnProp.PilgrimInstruments\")"))
  (bind _returnvalue_4 (Ability|GetAbilitySystemComponent (CallFunction|GetOwnerPlayerState)))
  (bind _returnvalue_5 (CallFunction|GetOwnerPlayerState))
  (bind _returnvalue_6 (CallFunction|GetOwnerPlayerState))
  (switch Utilities|FlowControl|Switch|SwitchonEPilgrimQuickplayState (CallFunction|GetPilgrimState)
    (:Loading
      _)
    (:Pregame
      _)
    (:Preintro
      (Utilities|IsValid _returnvalue_5 :!label _label
        (:Is%20Valid
          (Utilities|IsValid _returnvalue
            (:Is%20Valid
              (if (not (GameplayTags|HasMatchingGameplayTag _returnvalue _returnvalue_3))
                (Ability|GameplayEffect|AddLooseGameplayTags _returnvalue_5 (GameplayTags|MakeGameplayTagContainerfromTag _returnvalue_3))))))))
    (:Intro
      (goto _label))
    (:SongGameplay
      (goto _label))
    (:Outro
      (goto _label))
    (:SongResults
      (goto _label))
    (:SetResults
      (goto _label)))
  (Utilities|IsValid _returnvalue_6
    (:Is%20Valid
      (Utilities|IsValid _returnvalue_4
        (:Is%20Valid
          (if (GameplayTags|HasMatchingGameplayTag _returnvalue_4 _returnvalue_2)
            (Ability|GameplayEffect|RemoveLooseGameplayTags _returnvalue_6 (GameplayTags|MakeGameplayTagContainerfromTag _returnvalue_2))))))))

(fn UpdateTrackType
  (bind _returnvalue (CallFunction|GetOwnerPlayerState))
  (bind _returnvalue_2 (Actor|GetComponentbyClass _returnvalue "/Script/PilgrimCoreRuntime.PilgrimCorePlayerStateComponent"))
  (Utilities|IsValid _returnvalue
    (:Is%20Valid
      (Utilities|IsValid _returnvalue_2
        (:Is%20Valid
          (Variables|Default|SetTrackType (Class|PilgrimCorePlayerStateComponent|GetTrackTypeforGameplay _returnvalue_2))
          ;[1 0.5 0.5] Got the track type
          (return))
        (:Is%20Not%20Valid
          _)))
    (:Is%20Not%20Valid
      _))
  ;[1 0 0] Could not get the track type
  (return))
