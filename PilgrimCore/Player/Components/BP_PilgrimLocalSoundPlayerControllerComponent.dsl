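; Client-only player controller component that plays local 2D feedback sounds for Pilgrim music gameplay.
; On BeginPlay it loads per-instrument and default miss cues from the playspace's game config, then subscribes to GameplayEventRouter events.
; Handlers play sounds for misses (keyed by held instrument type), overdrive ready, deploy and end, same-team bandmate deploys, and stars gained.
; Score modulation events route to ApplyModulation, which has an empty body.
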(event EventBeginPlay
  (bind _returnvalue (Components|GetOwner))
  (bind _returnvalue_2 (Class|SparksMusicPlayspace|GetSparksMusicPlayspaceOfClass "/Script/PilgrimCoreRuntime.PilgrimPlayspace" _returnvalue))
  (bind _pilgrimgameconfig (Variables|/Script/PilgrimCoreRuntime.PilgrimPlayspace|GetPilgrimGameConfig _returnvalue_2))
  (bind _configsettings (Variables|/Script/PilgrimCoreRuntime.PilgrimGameConfig|GetConfigSettings _pilgrimgameconfig))
  (bind (_slopms _gembreakendinputunlockms _endoftrackbuffer _mstonowbar _backgroundtrackdb _backgroundtrackgain _songaudioresponcecuesettingsmap _firemidieventticksearly _basegemscorevalue _basecymbalgemscorevalue _sustainscoreperbeat _bnormalizescores _normalizedscaletowardshighestindifficulty _normalizedscaletowardshighestinexpert _outroprealertseconds _scoremultiplierdefinitions _playerhealthmodifiers _pilgrimhealthcrowdresponsetiers _individualstarpercentagethresholds _bandstarpercentagethresholds _isbandmultiplierbasedonuniqueinstruments _isbandvarietybonusscoredduringplay _bandoverdrivescoremultipliers _bandvarietyscoremultipliers _accuracytiers _maximumdisplayedaccuracytierindex _teamquickplayinitialhealth _teamquickplayoverdrivedecayrate _teamquickplayinitialoverdrive _teamquickplayoverdrive _teamquickplayoverdrivescoremultiplier _teamquickplayoverdrivehealthgainmultiplier _teamquickplayrequiredoverdrivetoactivate _teamquickplayoverdriveaddedonsuccesfulsection _boverdriveclearsgems _boverdriveclearautoholdsustain _gembreakpendingbars _benableinputspecificgemsandsmashers _inputconfig _benableoverdrivedeploys _benableoverdrivephrases _benablecapturesystem _benablebandperformancescoring _arenainitialshields _arenainitialhealth _arenainitialammo _difficultycapturefactors _perinstrumentconfigs _streaktoastthresholds _instrumenttypemisscues _defaultinstrumenttypemisscue) (Utilities|Struct|BreakPilgrimGameConfigSettings _configsettings))
  ;[1 0.4967 0.07286] This is meant to be client only
  (Networking|CanExecuteCosmeticEvents
    (:True
      (Utilities|FlowControl|Sequence
        (:then_0
          ;[0.2344 0.06035 0.2224] Set Instrument Miss Cues per instrument and the default from values in the pilgrim game config
          (Utilities|IsValid _returnvalue_2
            (:Is%20Valid
              (Variables|Default|SetInstrumentTypeMissCues _instrumenttypemisscues)
              (Variables|Default|SetDefaultMissSound _defaultinstrumenttypemisscue))))
        (:then_1
          (bind _eventrouter (Pilgrim|PilgrimWaitForEventRouter (Components|GetOwner))
            (:OnEventRouterAvailable
              (bind (_asynctaskproxy _payload) (GameplayEventRouter|StartListeningToEvent _eventrouter "/Script/PilgrimCoreRuntime.PilgrimEvent_LocalPlayerMiss" (CallFunction|GetOwningPlayerState))
                (:then
                  ;[1 0 0.8837] SFX Player captured an OD phrase.   Used to be OD Ready, when we needed at least 2 captures to be ready, which we don't now.
                  (bind (_asynctaskproxy_2 _payload_2) (GameplayEventRouter|StartListeningToEvent _eventrouter "/Script/PilgrimCoreRuntime.PilgrimEvent_Scoring_OverdriveReady" (CallFunction|GetOwningPlayerState))
                    (:then
                      ;[1 0 0.8837] Someone's OD deployed or depleted
                      (bind (_asynctaskproxy_3 _payload_3) (GameplayEventRouter|StartListeningToEvent _eventrouter "/Script/PilgrimCoreRuntime.PilgrimEvent_BandState_OverdriveActiveChanged" (CallFunction|GetOwningPlayerState))
                        (:then
                          ;[1 0 0.8837] Star Achieved
                          (bind (_asynctaskproxy_4 _payload_4) (GameplayEventRouter|StartListeningToEvent _eventrouter "/Script/PilgrimCoreRuntime.PilgrimEvent_Scoring_StarChanged" (CallFunction|GetOwningPlayerState))
                            (:OnEventReceived
                              (if (and (== (CallFunction|GetOwningPlayerState) (Utilities|Struct|BreakPilgrimEventScoringStarChanged _payload_4)) (> (.CurrentStar (Utilities|Struct|BreakPilgrimEventScoringStarChanged _payload_4)) (.PreviousStar (Utilities|Struct|BreakPilgrimEventScoringStarChanged _payload_4))))
                                (Audio|PlaySound2D "/SparksMaps/Audio/MSS_FM_Festival_Star_Achieved.MSS_FM_Festival_Star_Achieved")))))
                        (:OnEventReceived
                          (Default|HandleOverdriveActiveChanged (Utilities|Struct|BreakPilgrimEventBandStateOverdriveActiveChanged _payload_3) (Variables|/Script/Engine.Pawn|GetPlayerState (.ForPlayer (Utilities|Struct|BreakPilgrimEventBandStateOverdriveActiveChanged _payload_3)))))))
                    (:OnEventReceived
                      (if (== (CallFunction|GetOwningPlayerState) (Utilities|Struct|BreakPilgrimEventScoringOverdriveReady _payload_2))
                        (Audio|PlaySound2D "/SparksMaps/Audio/FM_od_ready.FM_od_ready")))))
                (:OnEventReceived
                  (CallFunction|PlayLocalPlayerMiss _payload))))))
        (:then_2
          (bind _eventrouter_2 (Pilgrim|PilgrimWaitForEventRouter (Components|GetOwner)))
          (bind (_asynctaskproxy_5 _payload_5) (GameplayEventRouter|StartListeningToEvent _eventrouter_2 "/Script/PilgrimCoreRuntime.PilgrimEvent_Scoring_Modulated" (CallFunction|GetOwningPlayerState))
            (:OnEventReceived
              (CallFunction|ApplyModulation (.CurrentModulation (Utilities|Struct|BreakPilgrimEventScoringModulated _payload_5))))))))))

(fn ApplyModulation
  (params (CurrentModulation :type Float%20%28double-precision%29) (CurrentInstrument :type Integer)))

(fn GetOwningPlayerState
  (bind _asplayer_controller (Utilities|Casting|CastToPlayerController (Components|GetOwner))
    (:then
      (return (Variables|/Script/Engine.Controller|GetPlayerState _asplayer_controller)))
    (:CastFailed
      (return 0))))

(fn HandleOverdriveActiveChanged
  (params (IsActive :type Boolean) (DeployingPlayer :type Object%20Reference))
  (if (== DeployingPlayer (CallFunction|GetOwningPlayerState))
    (if (Variables|GetIsActive)
      (Audio|PlaySound2D "/SparksMaps/Audio/MSS_FM_Overdrive_Deploy.MSS_FM_Overdrive_Deploy")
      (Audio|PlaySound2D "/SparksMaps/Audio/Crowd/MSS_Festival_Crowd_ODDeploy.MSS_Festival_Crowd_ODDeploy")
      (else
        (Audio|PlaySound2D "/SparksMaps/Audio/FM_od_end.FM_od_end")))
    (elif (Variables|GetIsActive)
      ;[1 0.9469 0.9569] Bandmate entered overdrive. Make sure the deploying player is on the same team as the owner.
      (bind _asfort_player_state (Utilities|Casting|CastToFortPlayerState (Variables|GetDeployingPlayer)))
      (if (== (Team|GetActorTeam _asfort_player_state) (Team|GetActorTeam (CallFunction|GetOwningPlayerState)))
        (Audio|PlaySound2D "/SparksMaps/Audio/FM_od_deploy_bandmate.FM_od_deploy_bandmate")))))

(fn PlayLocalPlayerMiss
  (params (PilgrimEvent_LocalPlayerMiss :type Pilgrim%20Event%20Local%20Player%20Miss%20Structure%20%28by%20ref%29))
  (if (== (Utilities|Struct|BreakPilgrimEventLocalPlayerMiss PilgrimEvent_LocalPlayerMiss) (CallFunction|GetOwningPlayerState))
    (bind _asfort_player_controller (Utilities|Casting|CastToFortPlayerController (Components|GetOwner)))
    (Utilities|IsValid (Variables|/Script/FortniteGame.FortPawn|GetCurrentWeapon (Game|Player|GetPlayerPawn _asfort_player_controller))
      (:Is%20Valid
        (Utilities|IsValid (Actor|GetComponentbyClass (Variables|/Script/FortniteGame.FortPawn|GetCurrentWeapon (Game|Player|GetPlayerPawn _asfort_player_controller)) "/SparksCosmetics/Components/BP_SparksInstrumentSettingsComponent.BP_SparksInstrumentSettingsComponent_C")
          (:Is%20Valid
            (Audio|PlaySound2D (select (.ReturnValue (Utilities|Map|Find (Variables|Default|GetInstrumentTypeMissCues) (Variables|/SparksCosmetics/Components/BP_SparksInstrumentSettingsComponent.BP_SparksInstrumentSettingsComponent_C|GetInstrumentType (Actor|GetComponentbyClass (Variables|/Script/FortniteGame.FortPawn|GetCurrentWeapon (Game|Player|GetPlayerPawn _asfort_player_controller)) "/SparksCosmetics/Components/BP_SparksInstrumentSettingsComponent.BP_SparksInstrumentSettingsComponent_C")))) (Utilities|Map|Find (Variables|Default|GetInstrumentTypeMissCues) (Variables|/SparksCosmetics/Components/BP_SparksInstrumentSettingsComponent.BP_SparksInstrumentSettingsComponent_C|GetInstrumentType (Actor|GetComponentbyClass (Variables|/Script/FortniteGame.FortPawn|GetCurrentWeapon (Game|Player|GetPlayerPawn _asfort_player_controller)) "/SparksCosmetics/Components/BP_SparksInstrumentSettingsComponent.BP_SparksInstrumentSettingsComponent_C"))) (Variables|Default|GetDefaultMissSound)) 1.0 1.0 0.0 "/SparksCharacterCommon/Sounds/ConcurrencyRules/Pilgrim_MissSfx_Conc.Pilgrim_MissSfx_Conc")))))))

(fn StopAndClearAll
  (params (Array :type Array%20of%20Audio%20Component%20Object%20References))
  (for _array_element Array
    (Audio|Components|Audio|Stop _array_element))
  (Utilities|Array|Clear Array))
