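; Player controller component for Pilgrim music gameplay, managing each player's instrument, difficulty, ready state and input mapping contexts.
; On local BeginPlay it waits for playlist, playspace and event router, restores saved choices, then listens for song, band health, quickplay and ready-up events.
; Server events write choices to the Pilgrim player state component.
; Playlist tags can force instrument swaps, and a next-tick flag drops duplicate ready toggles within a frame.
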(event Pilgrim|EventOnExitingSpectatingMode
  ; When spawning, we have to leave spectating mode to get the weapon/instrument inventory change
  (Development|PrintString "LogDumbBug: SetInstrumentOnSpawn being called from OnExitingSpectatingMode" false)
  (Class|BPPilgrimAltInstrumentControllerComponent|SetInstrumentonSpawn (Actor|GetComponentbyClass (Components|GetOwner) "/PilgrimCore/Player/Components/BP_PilgrimAltInstrumentControllerComponent.BP_PilgrimAltInstrumentControllerComponent_C")))

(event Custom|SetLoadedInstrumentAndDifficulty
  (bind _returnvalue (Components|GetOwner))
  (bind _returnvalue_2 (Actor|GetComponentbyClass _returnvalue "/PilgrimCore/Player/Components/BP_PilgrimAltInstrumentControllerComponent.BP_PilgrimAltInstrumentControllerComponent_C"))
  (bind _returnvalue_3 (CallFunction|GetPilgrimGameUserSettings))
  ;[0.589 1 0.9401] Initialize instrument and difficulty from saved PilgrimGameUserSettings
  (bind _asfort_player_controller (Utilities|Casting|CastToFortPlayerController (Components|GetOwner)))
  (CallFunction|ServerSetInstrumentandDifficulty (Class|PilgrimGameUserSettings|GetSelectedInstrumentforPlayer _returnvalue_3 _asfort_player_controller) (Class|PilgrimGameUserSettings|GetSongDifficultyforPlayer _returnvalue_3 _asfort_player_controller))
  (Class|BPPilgrimAltInstrumentControllerComponent|ServerSetTrackType _returnvalue_2 (Class|PilgrimGameUserSettings|GetSelectedInstrumentforPlayer _returnvalue_3 _asfort_player_controller))
  (Development|PrintString "LogDumbBug: SetInstrumentOnSpawn being called from BP_PilgrimControllerComponent" false)
  (Class|BPPilgrimAltInstrumentControllerComponent|SetInstrumentonSpawn _returnvalue_2))

(event Custom|ServerSetInstrumentAndDifficulty
  (params (LoadedInstrument :type EPilgrimTrackType%20Enum :default "TrackGuitar") (LoadedDifficulty :type EPilgrimSongDifficulty%20Enum :default "DifficultyEasy"))
  ;[0.589 1 0.9401] Initialize instrument and difficulty from saved PilgrimGameUserSettings
  (bind _asfort_player_controller (Utilities|Casting|CastToFortPlayerController (Components|GetOwner)))
  (bind _returnvalue (Class|PilgrimCorePlayerStateComponent|GetPilgrimCorePlayerStateComponent (Variables|/Script/Engine.Controller|GetPlayerState _asfort_player_controller)))
  (Class|PilgrimCorePlayerStateComponent|SetInstrumentChoice _returnvalue LoadedInstrument)
  (Class|PilgrimCorePlayerStateComponent|SetSongDifficulty _returnvalue LoadedDifficulty))

(event Custom|UnbindFromRespawn
  ;[1 1 1] Bind\Unbind to respawn (pawn possession)
  (bind _asfort_player_controller (Utilities|Casting|CastToFortPlayerController (Components|GetOwner)))
  (Events|UnbindEventfromOnPawnSet (Variables|/Script/Engine.Controller|GetPlayerState _asfort_player_controller) (delegate Custom|On%20New%20Pawn%20Set)))

(event Custom|ServerSetInstrumentByChoice
  (params (InstrumentSelected :type EPilgrimTrackType%20Enum :default "TrackGuitar"))
  ;[0.589 1 0.9401] Initialize instrument and difficulty from saved PilgrimGameUserSettings
  (bind _asfort_player_controller (Utilities|Casting|CastToFortPlayerController (Components|GetOwner)))
  (bind _returnvalue (Class|PilgrimCorePlayerStateComponent|GetPilgrimCorePlayerStateComponent (Variables|/Script/Engine.Controller|GetPlayerState _asfort_player_controller)))
  (Class|PilgrimCorePlayerStateComponent|SetInstrumentChoice _returnvalue InstrumentSelected))

(event EventEndPlay
  (params (EndPlayReason :type EEndPlayReason%20Enum :default "Destroyed"))
  (bind _eventlistenerhandles (Variables|Default|GetEventListenerHandles))
  (Fortnite|FortShippingLog 0 (Utilities|String|Append "FORT-567677: End Play Called for Pilgrim Player Controller Component for: " (Utilities|GetDisplayName (Components|GetOwner)) "   For Component:" (Utilities|GetDisplayName self) "   Endplay Reason:" (Utilities|String|EnumtoString EndPlayReason)))
  (Utilities|IsValid (Variables|Default|GetEventRouter)
    (:Is%20Valid
      (for _array_element _eventlistenerhandles
        (GameplayEventRouter|StopListeningtoEvent _array_element))
      (Utilities|Array|Clear _eventlistenerhandles))
    (:Is%20Not%20Valid
      _))
  (|CallParentFunction EndPlayReason))

(event Custom|BindToRespawn
  ;[1 1 1] Bind\Unbind to respawn (pawn possession)
  (bind _asfort_player_controller (Utilities|Casting|CastToFortPlayerController (Components|GetOwner)))
  (Events|BindEventtoOnPawnSet (Variables|/Script/Engine.Controller|GetPlayerState _asfort_player_controller) (delegate Custom|On%20New%20Pawn%20Set)))

(event Custom|On%20New%20Pawn%20Set
  (params (Player :type Player%20State%20Object%20Reference) (NewPawn :type Pawn%20Object%20Reference) (OldPawn :type Pawn%20Object%20Reference))
  ;[1 1 1] Bind\Unbind to respawn (pawn possession)
  (CallFunction|ActivateSelectedInputMappingContexts))

(event EventBeginPlay
  (bind _returnvalue (Components|GetOwner))
  (Fortnite|FortShippingLog 0 (Utilities|String|Append "FORT-567677: BeginPlay called on Controller's pilgrim component.  Controller: " (Utilities|GetDisplayName (Components|GetOwner)) "   For Component:" (Utilities|GetDisplayName self) "  Total Pilgrim Controller Components on Owner: " (Utilities|String|ToString(Integer) (Utilities|Array|Length (Actor|GetComponentsByClass (Components|GetOwner) "/PilgrimCore/Player/Components/BP_PilgrimPlayerControllerComponent.BP_PilgrimPlayerControllerComponent_C"))) "  Event Router is: " (Utilities|GetDisplayName (Variables|Default|GetEventRouter))))
  (|CallParentFunction)
  (bind (_gamestate _playlist) (Class|FortAsyncActionCurrentPlaylistReady|PlaylistDataReadyAsync)
    (:OnReady
      (bind _asfort_playlist_athena (Utilities|Casting|CastToFortPlaylistAthena _playlist))
      (CallFunction|SetupInstrumentFiltersfromPlaylistGameplayTags (Variables|/Script/FortniteGame.FortPlaylist|GetGameplayTagContainer _asfort_playlist_athena))
      (Variables|Default|SetPilgrimPlayspace (Class|SparksMusicPlayspace|GetSparksMusicPlayspaceOfClass "/Script/PilgrimCoreRuntime.PilgrimPlayspace" (Components|GetOwner)))
      ;[1 1 1] Setup Data Table and Debug Settings Widget
      (if (CallFunction|IsLocallyControlled)
        (bind (_asynctaskproxy _playspace) (Playspace|Discoverable|WaitForDiscoverablePlayspaceOfTypeForActor "/Script/SparksMusicPlayspaceRuntime.SparksMusicPlayspace" (Components|GetOwner))
          (:OnFoundPlayspace
            (Utilities|Casting|CastToSparksMusicPlayspace _playspace
              (:then
                (CallFunction|SetLoadedInstrumentandDifficulty)
                (bind _eventrouter (Pilgrim|PilgrimWaitForEventRouter (Components|GetOwner))
                  (:OnEventRouterAvailable
                    (Utilities|IsValid _eventrouter
                      (:Is%20Valid
                        (Variables|Default|SetEventRouter _eventrouter)
                        (Fortnite|FortShippingLog 0 (Utilities|String|Append "FORT-567677: Setting Event Router For Controller: " (Utilities|GetDisplayName (Components|GetOwner)) "   For Component:" (Utilities|GetDisplayName self) "  Total Pilgrim Controller Components on Owner: " (Utilities|String|ToString(Integer) (Utilities|Array|Length (Actor|GetComponentsByClass (Components|GetOwner) "/PilgrimCore/Player/Components/BP_PilgrimPlayerControllerComponent.BP_PilgrimPlayerControllerComponent_C"))) "  Event Router is valid: " (Utilities|String|ToString(Boolean) (Utilities|IsValid (Variables|Default|GetEventRouter) :!class /Script/Engine.KismetSystemLibrary))))
                        (Utilities|FlowControl|Sequence
                          (:then_0
                            (bind (_asynctaskproxy_2 _payload) (GameplayEventRouter|StartListeningToEvent (Variables|Default|GetEventRouter) "/Script/PilgrimCoreRuntime.PilgrimEvent_PilgrimGame_PlayingSong")
                              (:then
                                (GameplayEventRouter|StartListeningToEvent (Variables|Default|GetEventRouter) "/Script/PilgrimCoreRuntime.PilgrimEvent_PilgrimGame_SongStopped"
                                  (:then
                                    (bind (_asynctaskproxy_3 _payload_2) (GameplayEventRouter|StartListeningToEvent (Variables|Default|GetEventRouter) "/Script/PilgrimCoreRuntime.PilgrimEvent_BandState_BandHealthChanged")
                                      (:then
                                        ;[1 1 1] Make sure we allow peripherals if there is one connected now
                                        (InputDevices|BindEventtoOnInputHardwareDeviceChanged (EngineSubsystems|GetInputDeviceSubsystem) (delegate Custom|OnInputHardwareDeviceChange)))
                                      (:OnEventReceived
                                        (CallFunction|HandleBandHealthChanged _payload_2))))
                                  (:OnEventReceived
                                    (CallFunction|UnbindfromRespawn))))
                              (:OnEventReceived
                                (if (== (Class|PilgrimGame|GetLocalPlayerController (Utilities|Struct|BreakPilgrimEventPilgrimGamePlayingSong _payload)) (Variables|Default|GetOwningPlayerController))
                                  (CallFunction|ActivateSelectedInput _payload)
                                  (CallFunction|BindtoRespawn)))))
                          (:then_1
                            ;[0.6832 0.95 1] Add the GreenroomMappingContext initially
                            (CallFunction|ToggleGreenroomMappingContext true)
                            ;[0.5936 1 0.9901] When a new state begins, add the GreenroomMappingContext if it's Pregame, otherwise remove it
                            (bind (_asynctaskproxy_4 _payload_3) (GameplayEventRouter|StartListeningToEvent (Variables|Default|GetEventRouter) "/Script/PilgrimQPSharedRuntime.PilgrimEvent_QuickplayState_Began")
                              (:OnEventReceived
                                (Fortnite|FortShippingLog 0 (Utilities|String|Append "FORT-567677: Received PilgrimEvent QuickplayState_Began For Component: " (Utilities|GetDisplayName self) " : " (Utilities|String|EnumtoString (Utilities|Struct|BreakPilgrimEventQuickplayStateBase _payload_3))))
                                (CallFunction|ToggleGreenroomMappingContext (Utilities|Enum|Equal(Enum) (Utilities|Struct|BreakPilgrimEventQuickplayStateBase _payload_3) "Pregame")))))
                          (:then_2
                            (Fortnite|FortShippingLog 0 (Utilities|String|Append "FORT-567677: Starting listening for event TogglePlayerReadyButtonPressed for Controller: " (Utilities|GetDisplayName (Components|GetOwner)) "   For Component:" (Utilities|GetDisplayName self) "  Total Pilgrim Controller Components on Owner: " (Utilities|String|ToString(Integer) (Utilities|Array|Length (Actor|GetComponentsByClass (Components|GetOwner) "/PilgrimCore/Player/Components/BP_PilgrimPlayerControllerComponent.BP_PilgrimPlayerControllerComponent_C"))) "  Evemt Router is valid: " (Utilities|String|ToString(Boolean) (Utilities|IsValid (Variables|Default|GetEventRouter) :!class /Script/Engine.KismetSystemLibrary))))
                            (bind (_asynctaskproxy_5 _payload_4) (GameplayEventRouter|StartListeningToEvent (Variables|Default|GetEventRouter) "/PilgrimCore/Player/Components/PilgrimEvent_SongResult_TogglePlayerReadyButtonPressed.PilgrimEvent_SongResult_TogglePlayerReadyButtonPressed")
                              (:OnEventReceived
                                (Utilities|Time|SetTimerforNextTickbyEvent (delegate Custom|CustomEvent))
                                (if (Variables|Default|GetbIgnoringInputEvents)
                                  (Fortnite|FortShippingLog 0 (Utilities|String|Append "FORT-567677: Received Broad Casted Event TogglePlayerReadyButtonPressed - IGNORING additional events within the same frame: " (Utilities|GetDisplayName self)))
                                  (else
                                    (bind _payload_fortplayercontroller_11_9833968f4a2479a71e2b2ba53665a4e7 (Utilities|Struct|BreakPilgrimEventSongResultTogglePlayerReadyButtonPressed _payload_4))
                                    (Fortnite|FortShippingLog 0 (Utilities|String|Append "FORT-567677: Received Broad Casted Event TogglePlayerReadyButtonPressed sending Request to Server for Controller: " (Utilities|GetDisplayName _payload_fortplayercontroller_11_9833968f4a2479a71e2b2ba53665a4e7)))
                                    (CallFunction|RequestServerTogglePlayerReadyUp _payload_fortplayercontroller_11_9833968f4a2479a71e2b2ba53665a4e7)
                                    (Variables|Default|SetbIgnoringInputEvents true))))))
                          (:then_3
                            (Networking|CanExecuteCosmeticEvents
                              (:True
                                (bind _asfort_player_controller (Utilities|Casting|CastToFortPlayerController (Components|GetOwner)))
                                (Variables|Default|SetOwningPlayerController _asfort_player_controller))))
                          (:then_4
                            ;[1 1 1] Force Swap to instrument if some instruments types are unavailable.
                            (Fortnite|FortShippingLog 0 (Utilities|String|Append "Force Weapon Swap for " (Utilities|GetDisplayName (Components|GetOwner)) "   For Component:" (Utilities|GetDisplayName self) "(-1 means ignore/no swap) Initial Swap Value=" (Utilities|String|ToString(Integer) (Variables|Default|GetForceWeaponSwap))))
                            (if (>= (Variables|Default|GetForceWeaponSwap) 0)
                              (CallFunction|TrytoSwaptoInstrumentIndex (Variables|Default|GetForceWeaponSwap))))
                          (:then_5
                            ;[1 1 1] Pass along the fact that a player used the song selector to the server
                            (bind (_asynctaskproxy_6 _payload_5) (GameplayEventRouter|StartListeningToEvent (Variables|Default|GetEventRouter) "/Script/PilgrimCoreRuntime.PilgrimEvent_PlayerExperience_PlayerOpenedSongSelectMenu")
                              (:then
                                ;[0.05673 0.15 0.08398 0.5] Listen for the player to start and stop jamming. When they start, if we are showing the now playing UI, remove the part picker mapping context.
                                (if (CallFunction|IsUniversalPartPickerEnabled)
                                  (Utilities|FlowControl|Sequence
                                    (:then_0
                                      (Ability|Tasks|WaitGameplayTagAddToActor _returnvalue "(TagName=\"Gameplay.Action.Player.Jam.Jamming\")"
                                        (:Added
                                          (if (Class|PilgrimCoreUtilsLibrary|IsJamShowingNowPlayingUI)
                                            (bind _asfort_player_controller_2 (Utilities|Casting|CastToFortPlayerController (Components|GetOwner)))
                                            (if (.ReturnValue (Input|MappingQueries|HasMappingContext (PlayerController|LocalPlayerSubsystems|GetEnhancedInputLocalPlayerSubsystem _asfort_player_controller_2) "/PilgrimCore/Player/Abilities/IMC_PartPickerMappingContext.IMC_PartPickerMappingContext"))
                                              (Default|ToggleMappingContext false "/PilgrimCore/Player/Abilities/IMC_PartPickerMappingContext.IMC_PartPickerMappingContext" (CallFunction|GetGreenRoomMappingContextPriority)))))))
                                    (:then_1
                                      (Ability|Tasks|WaitGameplayTagRemoveFromActor _returnvalue "(TagName=\"Gameplay.Action.Player.Jam.Jamming\")"
                                        (:Removed
                                          (bind _asfort_player_controller_3 (Utilities|Casting|CastToFortPlayerController (Components|GetOwner)))
                                          (if (not (.ReturnValue (Input|MappingQueries|HasMappingContext (PlayerController|LocalPlayerSubsystems|GetEnhancedInputLocalPlayerSubsystem _asfort_player_controller_3) "/PilgrimCore/Player/Abilities/IMC_PartPickerMappingContext.IMC_PartPickerMappingContext")))
                                            (Default|ToggleMappingContext true "/PilgrimCore/Player/Abilities/IMC_PartPickerMappingContext.IMC_PartPickerMappingContext" (CallFunction|GetGreenRoomMappingContextPriority)))))))))
                              (:OnEventReceived
                                (CallFunction|ServerMarkPlayerHasInteractedwithSongSelector (Fortnite|GetFortPlayerControllerfromActor (Utilities|Struct|BreakPilgrimEventPlayerExperiencePlayerOpenedSongSelectMenu _payload_5))))))
                          (:then_6
                            (bind (_asynctaskproxy_7 _payload_6) (GameplayEventRouter|StartListeningToEvent (Variables|Default|GetEventRouter) "/Script/PilgrimCoreRuntime.PilgrimEvent_EnableChoppedCheeseCvarChanged")
                              (:OnEventReceived
                                (if (Utilities|Struct|BreakPilgrimEventEnableChoppedCheeseCvarChanged _payload_6)
                                  _
                                  (else
                                    _)))))
                          (:then_7
                            (bind (_asynctaskproxy_8 _payload_7) (GameplayEventRouter|StartListeningToEvent (Variables|Default|GetEventRouter) "/Script/PilgrimCoreRuntime.PilgrimEvent_EnableChoppedCheeseOverridePartCvarChanged")
                              (:OnEventReceived
                                (if (Utilities|Struct|BreakPilgrimEventEnableChoppedCheeseOverridePartCvarChanged _payload_7)
                                  _
                                  (else
                                    _)))))
                          (:then_8
                            ;[0.9888 1 0.8792 0.5] Update the Selected Instrument when Cymbal Settings Change to Force Switching Between Pro Drums and Pro Cymbals
                            (bind _asynctaskproxy_9 (GameplayEventRouter|StartListeningToEvent (Variables|Default|GetEventRouter) "/Script/PilgrimSettingsRuntime.PilgrimEvent_PilgrimSettings_ProDrumsEnableGreenCymbals")
                              (:then
                                (Utilities|Array|Add (Variables|Default|GetEventListenerHandles) _asynctaskproxy_9)
                                (bind _asynctaskproxy_10 (GameplayEventRouter|StartListeningToEvent (Variables|Default|GetEventRouter) "/Script/PilgrimSettingsRuntime.PilgrimEvent_PilgrimSettings_ProDrumsEnableRedCymbals")
                                  (:then
                                    (Utilities|Array|Add (Variables|Default|GetEventListenerHandles) _asynctaskproxy_10)
                                    (bind _asynctaskproxy_11 (GameplayEventRouter|StartListeningToEvent (Variables|Default|GetEventRouter) "/Script/PilgrimSettingsRuntime.PilgrimEvent_PilgrimSettings_ProDrumsEnableBlueCymbals")
                                      (:then
                                        (Utilities|Array|Add (Variables|Default|GetEventListenerHandles) _asynctaskproxy_11)
                                        (bind _asynctaskproxy_12 (GameplayEventRouter|StartListeningToEvent (Variables|Default|GetEventRouter) "/Script/PilgrimSettingsRuntime.PilgrimEvent_PilgrimSettings_ProDrumsEnableYellowCymbals")
                                          (:then
                                            (Utilities|Array|Add (Variables|Default|GetEventListenerHandles) _asynctaskproxy_12))
                                          (:OnEventReceived
                                            _)))
                                      (:OnEventReceived
                                        _)))
                                  (:OnEventReceived
                                    _)))
                              (:OnEventReceived
                                _))
                            (bind _asfort_player_controller_4 (Utilities|Casting|CastToFortPlayerController (Components|GetOwner)))
                            (bind _returnvalue_2 (Class|PilgrimCorePlayerStateComponent|GetPilgrimCorePlayerStateComponent (Variables|/Script/Engine.Controller|GetPlayerState _asfort_player_controller_4)))
                            (CallFunction|TrySwitchtoInstrument (Class|PilgrimCorePlayerStateComponent|GetInstrumentChoice _returnvalue_2))))
                        (bind _asfort_player_controller_5 (Utilities|Casting|CastToFortPlayerController (Components|GetOwner)))
                        (bind _returnvalue_3 (Class|PilgrimCorePlayerStateComponent|GetPilgrimCorePlayerStateComponent (Variables|/Script/Engine.Controller|GetPlayerState _asfort_player_controller_5)))
                        (CallFunction|TrytoSwaptoInstrumentIndex (Utilities|Array|FindItem (Variables|/Script/PilgrimCoreRuntime.PilgrimSupportedGameplayConfig|GetSupportedTrackTypes (Class|PilgrimPlayspace|GetSupportedGameplayConfig (Variables|Default|GetPilgrimPlayspace))) (Class|PilgrimCorePlayerStateComponent|GetInstrumentChoice _returnvalue_3))))
                      (:Is%20Not%20Valid
                        (Fortnite|FortShippingLog 0 (Utilities|String|Append "FORT-567677: Event Router is not valid from Pilgrim Wait Event Router for Controller: " (Utilities|GetDisplayName (Components|GetOwner)) "   For Component:" (Utilities|GetDisplayName self)) true))))))
              (:CastFailed
                (Fortnite|FortShippingLog self (Utilities|String|Append "WaitForPlayspace: Given playspace " (Utilities|GetDisplayName _playspace) " which doesn't derive from SparksMusicPlayspace.") true)))))
        (else
          (Fortnite|FortShippingLog 0 (Utilities|String|Append "FORT-567677: Controller is not locally controlled " (Utilities|GetDisplayName (Components|GetOwner)) "   For Component:" (Utilities|GetDisplayName self) "  Total Pilgrim Controller Components on Owner: ")))))))

(event Custom|OnInputHardwareDeviceChange
  (params (UserId :type Platform%20User%20Id%20Structure) (DeviceId :type Input%20Device%20Id%20Structure))
  ;[1 1 1] Make sure we allow peripherals if there is one connected now
  (CallFunction|CheckConnectedPeripherals))

(event EnhancedInputActionIA_PilgrimOpenCalibration
  (params (ActionValue :type Boolean) (ElapsedSeconds :type Float%20%28double-precision%29) (TriggeredSeconds :type Float%20%28double-precision%29) (InputAction :type Input%20Action%20Object%20Reference :default "/PilgrimCore/Input/InputActions/IA_PilgrimOpenCalibration.IA_PilgrimOpenCalibration")))

(event EnhancedInputActionIA_PilgrimDisengageBeatmatch
  (params (ActionValue :type Boolean) (ElapsedSeconds :type Float%20%28double-precision%29) (TriggeredSeconds :type Float%20%28double-precision%29) (InputAction :type Input%20Action%20Object%20Reference :default "/PilgrimSettings/Input/InputActions/IA_PilgrimDisengageBeatmatch.IA_PilgrimDisengageBeatmatch")))

(event EventOnPilgrimGameStart
  (params (PilgrimGame :type Pilgrim%20Game%20Object%20Reference))
  (|CallParentFunction PilgrimGame)
  (Default|CallOnSongStart))

(event EventOnPilgrimGameEnd
  (params (PilgrimGame :type Pilgrim%20Game%20Object%20Reference))
  (|CallParentFunction PilgrimGame)
  (Default|CallOnSongEnd))

(event Custom|CustomEvent
  (Variables|Default|SetbIgnoringInputEvents false))

(event Custom|ServerMarkPlayerHasInteractedWithSongSelector
  (params (Controller :type Fort%20Player%20Controller%20Object%20Reference))
  (bind _pilgrimplayspace (CallFunction|GetPilgrimPlayspace))
  (bind _returnvalue (Actor|GetComponentbyClass _pilgrimplayspace "/Script/PilgrimCoreRuntime.PilgrimPlayerManager"))
  (Utilities|IsValid _pilgrimplayspace
    (:Is%20Valid
      (Utilities|IsValid _returnvalue
        (:Is%20Valid
          (Class|PilgrimPlayerManager|MarkPlayerHasInteractedwithSongSelectorThisMatch _returnvalue Controller))))))

(event Custom|Request%20Server%20Toggle%20Player%20ReadyUp
  (params (Controller :type Fort%20Player%20Controller%20Object%20Reference))
  (bind _playerstate (Variables|/Script/Engine.Controller|GetPlayerState Controller))
  ;[1 1 1] Server request for Toggle player ready up
  (Utilities|IsValid Controller
    (:Is%20Valid
      (Utilities|IsValid _playerstate
        (:Is%20Valid
          (bind _returnvalue (Class|PilgrimCorePlayerStateComponent|GetPilgrimCorePlayerStateComponent _playerstate))
          (Utilities|IsValid _returnvalue
            (:Is%20Valid
              (Fortnite|FortShippingLog 0 (Utilities|String|Append "FORT-567677: RPC Setting PlayerState" (Utilities|GetDisplayName _returnvalue) " readiness from :" (Utilities|String|ToString(Boolean) (Class|PilgrimCorePlayerStateComponent|GetIsReady _returnvalue)) " to " (Utilities|String|ToString(Boolean) (not (Class|PilgrimCorePlayerStateComponent|GetIsReady _returnvalue)))))
              (Class|PilgrimCorePlayerStateComponent|ServerSetIsReady _returnvalue (not (Class|PilgrimCorePlayerStateComponent|GetIsReady _returnvalue))))
            (:Is%20Not%20Valid
              (Fortnite|FortShippingLog 0 "FORT-567677: Request for Toggle Ready Up RPC Pilgrim PlayerState is Invalid." true))))
        (:Is%20Not%20Valid
          (Fortnite|FortShippingLog 0 "FORT-567677: Request for Toggle Ready Up RPC PlayerState is Invalid. " true))))
    (:Is%20Not%20Valid
      (Fortnite|FortShippingLog 0 "FORT-567677: Request for Toggle Ready Up RPC Controller is Invalid." true))))

(event Custom|Server%20Set%20Player%20Ready%20Up
  (params (Controller :type Fort%20Player%20Controller%20Object%20Reference) (IsReady :type Boolean))
  (bind _playerstate (Variables|/Script/Engine.Controller|GetPlayerState Controller))
  ;[1 1 1] Server request for Toggle player ready up
  (Utilities|IsValid Controller
    (:Is%20Valid
      (Utilities|IsValid _playerstate
        (:Is%20Valid
          (bind _returnvalue (Class|PilgrimCorePlayerStateComponent|GetPilgrimCorePlayerStateComponent _playerstate))
          (Utilities|IsValid _returnvalue
            (:Is%20Valid
              (Class|PilgrimCorePlayerStateComponent|ServerSetIsReady _returnvalue IsReady))
            (:Is%20Not%20Valid
              (Fortnite|FortShippingLog 0 "FORT-567677: Request for Toggle Ready Up RPC Pilgrim PlayerState is Invalid." true))))
        (:Is%20Not%20Valid
          (Fortnite|FortShippingLog 0 "FORT-567677: Request for Toggle Ready Up RPC PlayerState is Invalid. " true))))
    (:Is%20Not%20Valid
      (Fortnite|FortShippingLog 0 "FORT-567677: Request for Toggle Ready Up RPC Controller is Invalid." true))))

(event Custom|SwapInstrumentForward
  (params (IsTouchInputPlatform :type Boolean))
  ;[0.15 0.15 0.15 0.5 13] Universal Part Picker temporary blocker
  (if (not (CallFunction|IsUniversalPartPickerEnabled))
    (Utilities|IsPlayerAllowedtoSelectInstruments
      (:True
        (bind _newinstrument (CallFunction|GetNextAvailableInstrument false IsTouchInputPlatform))
        (CallFunction|ServerTrySetInstrument _newinstrument)))))

(event EventSetIsReady
  (params (bIsReady :type Boolean))
  ;[1 1 1] Server request for Toggle player ready up
  (CallFunction|ServerSetPlayerReadyUp (Variables|Default|GetOwningPlayerController) bIsReady))

(event Custom|SwapInstrumentBackward
  ;[0.15 0.15 0.15 0.5 13] Universal Part Picker temporary blocker
  (if (not (CallFunction|IsUniversalPartPickerEnabled))
    (Utilities|IsPlayerAllowedtoSelectInstruments
      (:True
        (bind _newinstrument (CallFunction|GetNextAvailableInstrument true))
        (CallFunction|ServerTrySetInstrument _newinstrument)))))

(event EnhancedInputActionIA_PilgrimReturnToResults
  (params (ActionValue :type Boolean) (ElapsedSeconds :type Float%20%28double-precision%29) (TriggeredSeconds :type Float%20%28double-precision%29) (InputAction :type Input%20Action%20Object%20Reference :default "/PilgrimSettings/Player/Abilities/IA_PilgrimReturnToResults.IA_PilgrimReturnToResults")))

(event EnhancedInputActionIA_PilgrimSwapInstrument
  (params (ActionValue :type Boolean) (ElapsedSeconds :type Float%20%28double-precision%29) (TriggeredSeconds :type Float%20%28double-precision%29) (InputAction :type Input%20Action%20Object%20Reference :default "/PilgrimSettings/Player/Abilities/IA_PilgrimSwapInstrument.IA_PilgrimSwapInstrument")))

(event Custom|ServerTrySetInstrument
  (params (NewInstrument :type EPilgrimTrackType%20Enum :default "TrackGuitar"))
  ;[1 1 1] Green and Change Loadout Input Handling Events
  (CallFunction|ServerTrytoSetInstrument NewInstrument))

(event EventSetInstrument
  (params (NewInstrument :type EPilgrimTrackType%20Enum :default "TrackGuitar"))
  ;[1 1 1] Green and Change Loadout Input Handling Events
  (CallFunction|ServerTrySetInstrument NewInstrument))

(event EnhancedInputActionIA_PilgrimSwapInstrumentRight
  (params (ActionValue :type Boolean) (ElapsedSeconds :type Float%20%28double-precision%29) (TriggeredSeconds :type Float%20%28double-precision%29) (InputAction :type Input%20Action%20Object%20Reference :default "/PilgrimSettings/Player/Abilities/IA_PilgrimSwapInstrumentRight.IA_PilgrimSwapInstrumentRight")))

(event Custom|ServerTrySafeguardInstrument
  (params (NewInstrument :type EPilgrimTrackType%20Enum :default "TrackGuitar"))
  ;[1 1 1] Green and Change Loadout Input Handling Events
  (Default|ServerSafeguardInstrument NewInstrument))

(event EnhancedInputActionIA_PilgrimToggleReady
  (params (ActionValue :type Boolean) (ElapsedSeconds :type Float%20%28double-precision%29) (TriggeredSeconds :type Float%20%28double-precision%29) (InputAction :type Input%20Action%20Object%20Reference :default "/PilgrimSettings/Player/Abilities/IA_PilgrimToggleReady.IA_PilgrimToggleReady")))

(event EnhancedInputActionIA_PilgrimSwapInstrumentNumber1
  (params (ActionValue :type Boolean) (ElapsedSeconds :type Float%20%28double-precision%29) (TriggeredSeconds :type Float%20%28double-precision%29) (InputAction :type Input%20Action%20Object%20Reference :default "/PilgrimSettings/Player/Abilities/IA_PilgrimSwapInstrumentNumber1.IA_PilgrimSwapInstrumentNumber1")))

(event EnhancedInputActionIA_PilgrimSwapInstrumentNumber2
  (params (ActionValue :type Boolean) (ElapsedSeconds :type Float%20%28double-precision%29) (TriggeredSeconds :type Float%20%28double-precision%29) (InputAction :type Input%20Action%20Object%20Reference :default "/PilgrimSettings/Player/Abilities/IA_PilgrimSwapInstrumentNumber2.IA_PilgrimSwapInstrumentNumber2")))

(event EnhancedInputActionIA_PilgrimSwapDifficulty
  (params (ActionValue :type Boolean) (ElapsedSeconds :type Float%20%28double-precision%29) (TriggeredSeconds :type Float%20%28double-precision%29) (InputAction :type Input%20Action%20Object%20Reference :default "/PilgrimSettings/Player/Abilities/IA_PilgrimSwapDifficulty.IA_PilgrimSwapDifficulty")))

(event EnhancedInputActionIA_PilgrimSwapInstrumentNumber3
  (params (ActionValue :type Boolean) (ElapsedSeconds :type Float%20%28double-precision%29) (TriggeredSeconds :type Float%20%28double-precision%29) (InputAction :type Input%20Action%20Object%20Reference :default "/PilgrimSettings/Player/Abilities/IA_PilgrimSwapInstrumentNumber3.IA_PilgrimSwapInstrumentNumber3")))

(event EnhancedInputActionIA_PilgrimSwapInstrumentNumber4
  (params (ActionValue :type Boolean) (ElapsedSeconds :type Float%20%28double-precision%29) (TriggeredSeconds :type Float%20%28double-precision%29) (InputAction :type Input%20Action%20Object%20Reference :default "/PilgrimSettings/Player/Abilities/IA_PilgrimSwapInstrumentNumber4.IA_PilgrimSwapInstrumentNumber4")))

(event EnhancedInputActionIA_PilgrimSwapInstrumentNumber5
  (params (ActionValue :type Boolean) (ElapsedSeconds :type Float%20%28double-precision%29) (TriggeredSeconds :type Float%20%28double-precision%29) (InputAction :type Input%20Action%20Object%20Reference :default "/PilgrimSettings/Player/Abilities/IA_PilgrimSwapInstrumentNumber5.IA_PilgrimSwapInstrumentNumber5")))

(event Custom|Toggle%20Ready%20Input%20Pressed
  ;[1 1 1] Green and Change Loadout Input Handling Events
  (bind _asfort_player_controller (Utilities|Casting|CastToFortPlayerController (Components|GetOwner)))
  (bind _returnvalue (Class|PilgrimCorePlayerStateComponent|GetPilgrimCorePlayerStateComponent (Variables|/Script/Engine.Controller|GetPlayerState _asfort_player_controller)))
  (Class|PilgrimCorePlayerStateComponent|ServerSetIsReady _returnvalue (not (Class|PilgrimCorePlayerStateComponent|GetIsReady _returnvalue)))
  (CallFunction|MarkPlayerHasReadiedUp _asfort_player_controller))

(event EnhancedInputActionIA_PilgrimSwapInstrumentNumber6
  (params (ActionValue :type Boolean) (ElapsedSeconds :type Float%20%28double-precision%29) (TriggeredSeconds :type Float%20%28double-precision%29) (InputAction :type Input%20Action%20Object%20Reference :default "/PilgrimSettings/Player/Abilities/IA_PilgrimSwapInstrumentNumber6.IA_PilgrimSwapInstrumentNumber6")))

(event EnhancedInputActionIA_PilgrimSwapInstrumentNumber7
  (params (ActionValue :type Boolean) (ElapsedSeconds :type Float%20%28double-precision%29) (TriggeredSeconds :type Float%20%28double-precision%29) (InputAction :type Input%20Action%20Object%20Reference :default "/PilgrimSettings/Player/Abilities/IA_PilgrimSwapInstrumentNumber7.IA_PilgrimSwapInstrumentNumber7")))

(event Custom|Swap%20Difficulty%20Input%20Pressed
  (bind _returnvalue (Components|GetOwner))
  (bind _returnvalue_2 (Class|SparksMusicPlayspace|GetSparksMusicPlayspaceOfClass "/Script/PilgrimCoreRuntime.PilgrimPlayspace" _returnvalue))
  ;[1 1 1] Green and Change Loadout Input Handling Events
  (Variables|Default|GetPilgrimPlayspace
    (:then
      _)
    (:else
      (Utilities|IsValid _returnvalue_2
        (:Is%20Valid
          (Variables|Default|SetPilgrimPlayspace _returnvalue_2))
        (:Is%20Not%20Valid
          (Fortnite|FortShippingLog self "WaitForPlayspace: Couldn't find the playspace despite waiting for the playspace before binding inputs" true)))))
  (bind _asfort_player_controller (Utilities|Casting|CastToFortPlayerController _returnvalue))
  (bind _returnvalue_3 (Class|PilgrimCorePlayerStateComponent|GetPilgrimCorePlayerStateComponent (Variables|/Script/Engine.Controller|GetPlayerState _asfort_player_controller)))
  (CallFunction|SetDifficultyModetoNextAvailable _returnvalue_3 (Class|PilgrimCorePlayerStateComponent|GetSongDifficulty _returnvalue_3)))

(event EnhancedInputActionIA_PilgrimSwapInstrumentNumber8
  (params (ActionValue :type Boolean) (ElapsedSeconds :type Float%20%28double-precision%29) (TriggeredSeconds :type Float%20%28double-precision%29) (InputAction :type Input%20Action%20Object%20Reference :default "/PilgrimSettings/Player/Abilities/IA_PilgrimSwapInstrumentNumber8.IA_PilgrimSwapInstrumentNumber8")))

(event EnhancedInputActionIA_PilgrimPartPickerMenuUp
  (params (ActionValue :type Boolean) (ElapsedSeconds :type Float%20%28double-precision%29) (TriggeredSeconds :type Float%20%28double-precision%29) (InputAction :type Input%20Action%20Object%20Reference :default "/PilgrimCore/Player/Abilities/IA_PilgrimPartPickerMenuUp.IA_PilgrimPartPickerMenuUp")))

(event Custom|ServerSetDifficulty
  (params (NewDifficulty :type EPilgrimSongDifficulty%20Enum :default "DifficultyEasy"))
  ;[1 1 1] Green and Change Loadout Input Handling Events
  (bind _asfort_player_controller (Utilities|Casting|CastToFortPlayerController (Components|GetOwner)))
  (bind _returnvalue (Class|PilgrimCorePlayerStateComponent|GetPilgrimCorePlayerStateComponent (Variables|/Script/Engine.Controller|GetPlayerState _asfort_player_controller)))
  (Class|PilgrimCorePlayerStateComponent|SetSongDifficulty _returnvalue NewDifficulty))

(event EventSetDifficulty
  (params (NewDifficulty :type EPilgrimSongDifficulty%20Enum :default "DifficultyEasy"))
  (CallFunction|ServerSetDifficulty NewDifficulty))

(event Custom|Try%20to%20Select%20Difficulty
  (params (NewDifficulty :type EPilgrimSongDifficulty%20Enum :default "DifficultyEasy"))
  ; New logic to get a specific difficulty, used by Universal Part Picker
  (bind _asfort_player_controller (Utilities|Casting|CastToFortPlayerController (Components|GetOwner)))
  (bind _returnvalue (Class|PilgrimCorePlayerStateComponent|GetPilgrimCorePlayerStateComponent (Variables|/Script/Engine.Controller|GetPlayerState _asfort_player_controller)))
  (CallFunction|SetSpecificDifficultyMode _returnvalue NewDifficulty))

(event Custom|SafeguardInstrumentChoice
  (bind _pilgrimplayspace (CallFunction|GetPilgrimPlayspace))
  (bind _returnvalue (Class|PilgrimPlayspace|GetSupportedGameplayConfig _pilgrimplayspace))
  (bind _supportedtracktypes (Variables|/Script/PilgrimCoreRuntime.PilgrimSupportedGameplayConfig|GetSupportedTrackTypes _returnvalue))
  ;[1 1 1] Safeguard Instument swap - Added to prevent players from starting with the wrong instruments in Battlestage Competitive modes.
  (bind (_gamestate _playlist) (Class|FortAsyncActionCurrentPlaylistReady|PlaylistDataReadyAsync)
    (:OnReady
      (bind _asfort_playlist_athena (Utilities|Casting|CastToFortPlaylistAthena _playlist))
      (CallFunction|SetupInstrumentFiltersfromPlaylistGameplayTags (Variables|/Script/FortniteGame.FortPlaylist|GetGameplayTagContainer _asfort_playlist_athena))
      (if (>= (Variables|Default|GetForceWeaponSwap) 0)
        (Fortnite|FortShippingLog 0 (Utilities|String|Append "Force Weapon Swap for " (Utilities|GetDisplayName (Components|GetOwner)) "   For Component:" (Utilities|GetDisplayName self) " SafeguardInstrumentChoice: Value=" (Utilities|String|ToString(Integer) (Variables|Default|GetForceWeaponSwap))))
        (bind _etracktype (Default|GetTrackTypesEnumfromIndex (Variables|Default|GetForceWeaponSwap)))
        (if (Utilities|Array|IsValidIndex _supportedtracktypes (Utilities|Array|FindItem _supportedtracktypes _etracktype))
          (CallFunction|ServerTrySafeguardInstrument (Utilities|Array|Get(aref) _supportedtracktypes (Utilities|Array|FindItem _supportedtracktypes _etracktype))))))))

(fn ActivateSelectedInput
  (params (PilgrimEvent_PilgrimGame_PlayingSong :type Pilgrim%20Event%20Pilgrim%20Game%20Playing%20Song%20Structure%20%28by%20ref%29))
  (CallFunction|SetNumInputLanes self (Math|Conversions|ToByte(Integer) (Class|PilgrimGame|GetNumLanes (Utilities|Struct|BreakPilgrimEventPilgrimGamePlayingSong PilgrimEvent_PilgrimGame_PlayingSong)))))

(fn CheckForAcceptingInput
  (params (Name :type String) (GameplayTag :type Gameplay%20Tag%20Structure) (DisplayDebug :type Boolean))
  (LocalVariables|SetLclDisplayDebug DisplayDebug)
  (LocalVariables|SetLclName Name)
  (if (Variables|Default|GetbUseEnhancedInput)
    (bind _asfort_player_controller (Utilities|Casting|CastToFortPlayerController (Components|GetOwner))
      (:then
        (Utilities|IsValid (Game|Player|GetPlayerPawn _asfort_player_controller)
          (:Is%20Valid
            (Utilities|IsValid (Ability|GetAbilitySystemComponent (Game|Player|GetPlayerPawn _asfort_player_controller))
              (:Is%20Valid
                (if (GameplayTags|HasMatchingGameplayTag (Ability|GetAbilitySystemComponent (Game|Player|GetPlayerPawn _asfort_player_controller)) GameplayTag)
                  (Utilities|IsValid (Variables|Default|GetEventRouter)
                    (:Is%20Valid
                      (Development|PrintString (Utilities|String|Append (LocalVariables|GetLclName) ": Enhanced Input succeeded.") (LocalVariables|GetLclDisplayDebug) (LocalVariables|GetLclDisplayDebug))
                      (return true))
                    (:Is%20Not%20Valid
                      (Development|PrintString (Utilities|String|Append (LocalVariables|GetLclName) ":  event router is invalid.") (LocalVariables|GetLclDisplayDebug) (LocalVariables|GetLclDisplayDebug) "(R=1.000000,G=0.000000,B=0.055182,A=1.000000)")))
                  (else
                    (Development|PrintString (Utilities|String|Append (LocalVariables|GetLclName) ": tag is not present.  Tag is " (Utilities|String|ToString(Name) (GameplayTags|GetTagName GameplayTag))) (LocalVariables|GetLclDisplayDebug) (LocalVariables|GetLclDisplayDebug) "(R=1.000000,G=0.000000,B=0.055182,A=1.000000)"))))
              (:Is%20Not%20Valid
                (Development|PrintString (Utilities|String|Append (LocalVariables|GetLclName) ": controller->pawn->abilisySysComp is invalid") (LocalVariables|GetLclDisplayDebug) (LocalVariables|GetLclDisplayDebug) "(R=1.000000,G=0.000000,B=0.055182,A=1.000000)"))))
          (:Is%20Not%20Valid
            (Development|PrintString (Utilities|String|Append (LocalVariables|GetLclName) ": controller->pawn is invalid") (LocalVariables|GetLclDisplayDebug) (LocalVariables|GetLclDisplayDebug) "(R=1.000000,G=0.000000,B=0.055182,A=1.000000)"))))
      (:CastFailed
        (Development|PrintString (Utilities|String|Append (LocalVariables|GetLclName) ": controller is invalid") (LocalVariables|GetLclDisplayDebug) (LocalVariables|GetLclDisplayDebug) "(R=1.000000,G=0.000000,B=0.055182,A=1.000000)")))
    (return false)
    (else
      (Development|PrintString (Utilities|String|Append (LocalVariables|GetLclName) ": bUseEnhancedInput is false") (LocalVariables|GetLclDisplayDebug) (LocalVariables|GetLclDisplayDebug) "(R=1.000000,G=0.000000,B=0.055182,A=1.000000)"))))

(fn EnsureValidIndex
  (params (IncomingValue :type Integer) (HasConnectedGuitar :type Boolean) (CycleRight :type Boolean))
  (bind _output_get (LocalVariables|SetLclOriginalIndex IncomingValue))
  (LocalVariables|SetLclCurrentValue _output_get)
  (switch int (LocalVariables|GetLclCurrentValue) :!label _label
    (:Default
      _)
    (:0
      (if (Variables|InstrumentFilters|GetbAllowVocalsSelect)
        (return (LocalVariables|GetLclCurrentValue))
        (else
          _)))
    (:1
      (if (Variables|InstrumentFilters|GetbAllowBassSelect)
        (return (LocalVariables|GetLclCurrentValue))
        (else
          _)))
    (:2
      (if (Variables|InstrumentFilters|GetbAllowGuitarSelect)
        (return (LocalVariables|GetLclCurrentValue))
        (else
          _)))
    (:3
      (if (Variables|InstrumentFilters|GetbAllowDrumsSelect)
        (return (LocalVariables|GetLclCurrentValue))
        (else
          _)))
    (:4
      (if (Variables|GetHasConnectedGuitar)
        (if (Variables|InstrumentFilters|GetbAllowGuitarSelect)
          (return (LocalVariables|GetLclCurrentValue))
          (else
            _))))
    (:5
      (if (Variables|GetHasConnectedGuitar)
        (if (Variables|InstrumentFilters|GetbAllowBassSelect)
          (return (LocalVariables|GetLclCurrentValue))
          (else
            _)))))
  ;[1 1 1] Increment and Wrap around if needed
  (if (Variables|GetCycleRight)
    (Math|Integer|IncrementInt (LocalVariables|GetLclCurrentValue)
      (:%20%20
        (if (> (LocalVariables|GetLclCurrentValue) 5)
          (LocalVariables|SetLclCurrentValue 0))))
    (else
      (Math|Integer|DecrementInt (LocalVariables|GetLclCurrentValue)
        (:%20%20
          (if (< (LocalVariables|GetLclCurrentValue) 0)
            (LocalVariables|SetLclCurrentValue 5))))))
  ;[1 1 1] If we wrapped all the way around to where we started, just return the original (couldn't find a valid instrument)
  (if (== (LocalVariables|GetLclCurrentValue) (LocalVariables|GetLclOriginalIndex))
    (return (LocalVariables|GetLclOriginalIndex))
    (else
      (goto _label))))

(fn GetDividerIndex
  (bind _returnvalue (CallFunction|GetSelectedPreset))
  (bind _diividerpositionsbypreset (Variables|Default|GetDiividerPositionsByPreset))
  (bind (_value _returnvalue_2) (Utilities|Map|Find _diividerpositionsbypreset _returnvalue))
  (if (Default|IsUsingMouseandKeyboard)
    (Utilities|IsValid _returnvalue
      (:Is%20Valid
        (if _returnvalue_2
          ;[1 1 1] Specified hand separation
          (return _value)
          (else
            _)))
      (:Is%20Not%20Valid
        ;[1 1 1] Custom mappings
        (return -1))))
  ;[1 1 1] Default
  (return 2))

(fn GetFirstAvailableIndex
  (if (Variables|InstrumentFilters|GetbAllowVocalsSelect)
    (return 0)
    (elif (Variables|InstrumentFilters|GetbAllowBassSelect)
      (return 1)
      (elif (Variables|InstrumentFilters|GetbAllowGuitarSelect)
        (return 2)
        (elif (Variables|InstrumentFilters|GetbAllowDrumsSelect)
          _))))
  (return 3))

(fn GetLastAvailableIndex
  (params (HasConnectedGuitar :type Boolean))
  (Utilities|FlowControl|Sequence
    (:then_0
      (if (Variables|InstrumentFilters|GetbAllowDrumsSelect)
        (LocalVariables|SetLastAvailableNonPeripheral 3)
        (elif (Variables|InstrumentFilters|GetbAllowGuitarSelect)
          (LocalVariables|SetLastAvailableNonPeripheral 2)
          (elif (Variables|InstrumentFilters|GetbAllowBassSelect)
            (LocalVariables|SetLastAvailableNonPeripheral 1)
            (elif (Variables|InstrumentFilters|GetbAllowVocalsSelect)
              (LocalVariables|SetLastAvailableNonPeripheral 0))))))
    (:then_1
      (if (Variables|InstrumentFilters|GetbAllowBassSelect)
        (return (Math|Integer|SelectInt 5 (LocalVariables|GetLastAvailableNonPeripheral) HasConnectedGuitar))
        (elif (Variables|InstrumentFilters|GetbAllowGuitarSelect)
          (return (Math|Integer|SelectInt 4 (LocalVariables|GetLastAvailableNonPeripheral) HasConnectedGuitar))
          (else
            (return (LocalVariables|GetLastAvailableNonPeripheral))))))))

(fn GetNextAvailableInstrument
  (params (Backwards :type Boolean) (IsTouchInputPlatform :type Boolean))
  (bind _newindex (LocalVariables|GetNewIndex))
  (LocalVariables|SetStep (select Backwards -1 1))
  (bind _asfort_player_controller (Utilities|Casting|CastToFortPlayerController (Components|GetOwner))
    (:then
      (bind _returnvalue (Class|PilgrimCorePlayerStateComponent|GetPilgrimCorePlayerStateComponent (Variables|/Script/Engine.Controller|GetPlayerState _asfort_player_controller)))
      (LocalVariables|SetCurrentInstrument (Class|PilgrimCorePlayerStateComponent|GetInstrumentChoice _returnvalue))
      (bind _output_get (LocalVariables|SetSupportedInstruments (select (Variables|GetIsTouchInputPlatform) (Variables|/Script/PilgrimCoreRuntime.PilgrimSupportedGameplayConfig|GetTouchInputSupportedTrackTypes (Class|PilgrimPlayspace|GetSupportedGameplayConfig (Variables|Default|GetPilgrimPlayspace))) (Variables|/Script/PilgrimCoreRuntime.PilgrimSupportedGameplayConfig|GetSupportedTrackTypes (Class|PilgrimPlayspace|GetSupportedGameplayConfig (Variables|Default|GetPilgrimPlayspace))))))
      (Utilities|FlowControl|Sequence
        (:then_0
          (Utilities|Array|RemoveItem _output_get "TrackProCymbals"))
        (:then_1
          (if (not (CallFunction|IsChoppedCheeseEnabled))
            (Utilities|Array|RemoveItem _output_get "TrackPlasticDrum")))
        (:then_2
          (if (not (CallFunction|IsChickenParmEnabled))
            (Utilities|Array|RemoveItem _output_get "TrackProVocals")))
        (:then_3
          (bind _output_get_2 (LocalVariables|SetCurrentIndex (Utilities|Array|FindItem (LocalVariables|GetSupportedInstruments) (LocalVariables|GetCurrentInstrument))))
          (if (>= _output_get_2 0)
            (LocalVariables|SetNewIndex (+ (LocalVariables|GetCurrentIndex) (LocalVariables|GetStep)))
            (if (CallFunction|CanProInstrumentsbeSelected)
              (LocalVariables|SetMaxWrapIndex (- (Utilities|Array|Length (LocalVariables|GetSupportedInstruments)) 1))
              (else
                (LocalVariables|SetMaxWrapIndex 3)))
            (if (and (>= (LocalVariables|GetNewIndex) 0) (<= (LocalVariables|GetNewIndex) (LocalVariables|GetMaxWrapIndex)))
              _
              (elif (< _newindex 0)
                ;[0.143 0 0.15 0.5] Wrapping index
                (LocalVariables|SetNewIndex (LocalVariables|GetMaxWrapIndex))
                (if (not (Utilities|Array|IsValidIndex (LocalVariables|GetSupportedInstruments) (LocalVariables|GetNewIndex))) :!label _label
                  (if (> (Utilities|Array|Length (LocalVariables|GetSupportedInstruments)) 0) :!label _label_2
                    ;[0.1475 0.15 0.04862 0.5] If something goes wrong, fallback to the first available instrument
                    (return (Utilities|Array|Get(aref) (LocalVariables|GetSupportedInstruments)))
                    (else
                      ;[0.15 0.02812 0.02812 0.5] if something goes seriously wrong, just keep the same instrument
                      (return (LocalVariables|GetCurrentInstrument)))))
                (elif (> _newindex (LocalVariables|GetMaxWrapIndex))
                  (LocalVariables|SetNewIndex 0)
                  (goto _label)
                  (else
                    (goto _label)))))
            (return (Utilities|Array|Get(acopy) (LocalVariables|GetSupportedInstruments) (LocalVariables|GetNewIndex))))
          (goto _label_2))))
    (:CastFailed
      (return "TrackGuitar"))))

(fn GetPilgrimPlayspace
  (bind _returnvalue (Components|GetOwner))
  (bind _returnvalue_2 (Class|SparksMusicPlayspace|GetSparksMusicPlayspaceOfClass "/Script/PilgrimCoreRuntime.PilgrimPlayspace" _returnvalue))
  (Utilities|IsValid (Variables|Default|GetPilgrimPlayspace)
    (:Is%20Valid
      (return (Variables|Default|GetPilgrimPlayspace)))
    (:Is%20Not%20Valid
      (Utilities|IsValid _returnvalue_2
        (:Is%20Valid
          (bind _output_get (Variables|Default|SetPilgrimPlayspace _returnvalue_2))
          (return _output_get))
        (:Is%20Not%20Valid
          (Fortnite|FortShippingLog self "BP_PilgrimPlayerControllerComponent: Couldn't fetch Pilgrim playspace." true)
          (return 0))))))

(fn GetTrackTypesEnumFromIndex
  (params (Index :type Integer))
  ;[0.1307 0.15 0.07545 0.5] Within PilgrimGameConfig.h
  ; See:  SupportedTrackTypes
  (switch int Index
    (:Default
      ;; Default Return
      (return "TrackNone"))
    (:0
      (return "TrackVocals"))
    (:1
      (return "TrackBass"))
    (:2
      (return "TrackGuitar"))
    (:3
      (return "TrackDrum"))
    (:4
      (return "TrackPlasticGuitar"))
    (:5
      (return "TrackPlasticBass"))
    (:6
      (return "TrackPlasticDrum"))
    (:7
      (return "TrackProVocals"))))

(fn HandleBandHealthChanged
  (params (HealthChangedEvent :type Pilgrim%20Event%20Band%20State%20Band%20Health%20Changed%20Structure%20%28by%20ref%29))
  (bind (_lastbandhealth _newbandhealth _lastbandhealthtier _newbandhealthtier) (Utilities|Struct|BreakPilgrimEventBandStateBandHealthChanged HealthChangedEvent))
  (CallFunction|UpdateCrowdResponseHealthTiers HealthChangedEvent)
  (if (!= _lastbandhealthtier _newbandhealthtier)
    (if (> (CallFunction|TimeSinceLastCrowdResponseChange) (Variables|Default|GetMinimumTimeBetweenCrowdResponseChanges))
      (CallFunction|UpdateCrowdResponse)
      (else
        (Utilities|Time|SetTimerbyEvent (EventDispatchers|CreateEvent :!event Update%20Crowd%20Response) (Variables|Default|GetMinimumTimeBetweenCrowdResponseChanges))))))

(fn IsInstrumentAllowed
  (params (Instrument :type EPilgrimTrackType%20Enum :default "TrackGuitar"))
  (return (Utilities|Array|ContainsItem (Variables|/Script/PilgrimCoreRuntime.PilgrimSupportedGameplayConfig|GetSupportedTrackTypes (Class|PilgrimPlayspace|GetSupportedGameplayConfig (Variables|Default|GetPilgrimPlayspace))) Instrument)))

(fn IsUsingMouseAndKeyboard
  (bind _asfort_player_controller (Utilities|Casting|CastToFortPlayerController (Components|GetOwner))
    (:then
      (return (Utilities|Enum|Equal(Enum) (CommonInputSubsystem|GetCurrentInputType (PlayerController|LocalPlayerSubsystems|GetCommonInputSubsystem _asfort_player_controller)) "MouseAndKeyboard")))
    (:CastFailed
      (return false))))

(fn MarkPlayerHasReadiedUp
  (params (PlayerController :type Fort%20Player%20Controller%20Object%20Reference))
  (bind _pilgrimplayspace (CallFunction|GetPilgrimPlayspace))
  (bind _returnvalue (Actor|GetComponentbyClass _pilgrimplayspace "/Script/PilgrimCoreRuntime.PilgrimPlayerManager"))
  (Utilities|IsValid _pilgrimplayspace
    (:Is%20Valid
      (Utilities|IsValid _returnvalue
        (:Is%20Valid
          (Class|PilgrimPlayerManager|MarkPlayerHasReadiedUp _returnvalue PlayerController))))))

(fn ServerOnly_CalcHealthForTeam)

(fn ServerSafeguardInstrument
  (params (NewInstrument :type EPilgrimTrackType%20Enum :default "TrackGuitar"))
  (Fortnite|FortShippingLog 0 (Utilities|String|Append "Force Weapon Swap for " (Utilities|GetDisplayName (Components|GetOwner)) "   For Component:" (Utilities|GetDisplayName self) " ServerSafeguardInstrument: Value=" (Utilities|String|EnumtoString NewInstrument)))
  (bind _asfort_player_controller (Utilities|Casting|CastToFortPlayerController (Components|GetOwner)))
  (bind _returnvalue (Class|PilgrimCorePlayerStateComponent|GetPilgrimCorePlayerStateComponent (Variables|/Script/Engine.Controller|GetPlayerState _asfort_player_controller)))
  (Utilities|IsValid _returnvalue
    (:Is%20Valid
      (LocalVariables|SetLclPilgrimCorePlayerStateComp _returnvalue)
      ;[1 1 1] Verify Instrument is valid
      (if (Utilities|Array|IsValidIndex (Variables|/Script/PilgrimCoreRuntime.PilgrimSupportedGameplayConfig|GetSupportedTrackTypes (Class|PilgrimPlayspace|GetSupportedGameplayConfig (Variables|Default|GetPilgrimPlayspace))) (Utilities|Array|FindItem (Variables|/Script/PilgrimCoreRuntime.PilgrimSupportedGameplayConfig|GetSupportedTrackTypes (Class|PilgrimPlayspace|GetSupportedGameplayConfig (Variables|Default|GetPilgrimPlayspace))) (Variables|GetNewInstrument)))
        ;[1 1 1] For Safeguard, do not override/switch off plastics if that's what's selected.
        (bind _returnvalue_2 (Class|PilgrimCorePlayerStateComponent|GetInstrumentChoice (LocalVariables|GetLclPilgrimCorePlayerStateComp)))
        (switch Utilities|FlowControl|Switch|SwitchonEPilgrimTrackType (Variables|GetNewInstrument)
          (:TrackGuitar
            (if (Utilities|Enum|NotEqual(Enum) _returnvalue_2 "TrackPlasticGuitar")
              _))
          (:TrackBass
            (if (Utilities|Enum|NotEqual(Enum) _returnvalue_2 "TrackPlasticBass")
              _))
          (:TrackVocals
            (if (Utilities|Enum|NotEqual(Enum) _returnvalue_2 "TrackProVocals")
              _))
          (:TrackDrum
            (if (or (Utilities|Enum|NotEqual(Enum) _returnvalue_2 "TrackPlasticDrum") (Utilities|Enum|NotEqual(Enum) _returnvalue_2 "TrackProCymbals"))
              _)))
        (Class|PilgrimCorePlayerStateComponent|SetInstrumentChoice (LocalVariables|GetLclPilgrimCorePlayerStateComp) (Variables|GetNewInstrument))))))

(fn ServerTryToSetInstrument
  (params (NewInstrument :type EPilgrimTrackType%20Enum :default "TrackGuitar"))
  (Utilities|IsPlayerAllowedtoSelectInstruments
    (:True
      (bind _asfort_player_controller (Utilities|Casting|CastToFortPlayerController (Components|GetOwner)))
      (bind _returnvalue (Class|PilgrimCorePlayerStateComponent|GetPilgrimCorePlayerStateComponent (Variables|/Script/Engine.Controller|GetPlayerState _asfort_player_controller)))
      (Utilities|IsValid _returnvalue
        (:Is%20Valid
          ;[1 1 1] Verify Instrument is valid
          (if (Utilities|Array|IsValidIndex (Variables|/Script/PilgrimCoreRuntime.PilgrimSupportedGameplayConfig|GetSupportedTrackTypes (Class|PilgrimPlayspace|GetSupportedGameplayConfig (Variables|Default|GetPilgrimPlayspace))) (Utilities|Array|FindItem (Variables|/Script/PilgrimCoreRuntime.PilgrimSupportedGameplayConfig|GetSupportedTrackTypes (Class|PilgrimPlayspace|GetSupportedGameplayConfig (Variables|Default|GetPilgrimPlayspace))) (Variables|GetNewInstrument)))
            (Class|PilgrimCorePlayerStateComponent|SetInstrumentChoice _returnvalue (Variables|GetNewInstrument))))))
    (:False
      (Development|PrintString "Hello"))))

(fn SetDifficultyModeToNextAvailable
  (params (PilgrimCorePlayerStateComponent :type Pilgrim%20Core%20Player%20State%20Component%20Object%20Reference) (InCurrentDifficulty :type EPilgrimSongDifficulty%20Enum :default "DifficultyEasy"))
  (bind _pilgrimplayspace (Variables|Default|GetPilgrimPlayspace))
  (bind _returnvalue (Class|PilgrimPlayspace|GetSupportedGameplayConfig _pilgrimplayspace))
  (bind _supporteddifficulties (Variables|/Script/PilgrimCoreRuntime.PilgrimSupportedGameplayConfig|GetSupportedDifficulties _returnvalue))
  (bind _returnvalue_2 (Utilities|Array|FindItem _supporteddifficulties InCurrentDifficulty))
  (bind _returnvalue_3 (+ _returnvalue_2 1))
  (if (Utilities|Array|IsValidIndex _supporteddifficulties _returnvalue_3)
    (Class|PilgrimCorePlayerStateComponent|SetSongDifficulty PilgrimCorePlayerStateComponent (Utilities|Array|Get(aref) _supporteddifficulties _returnvalue_3))
    (elif (Utilities|Array|IsValidIndex _supporteddifficulties)
      (Class|PilgrimCorePlayerStateComponent|SetSongDifficulty PilgrimCorePlayerStateComponent (Utilities|Array|Get(aref) _supporteddifficulties))
      (else
        (Development|PrintString "ERROR: Supported Difficulties Array is empty. Default will be Easy." true true "(R=1.000000,G=0.028928,B=0.000000,A=1.000000)" 2e+02)))))

(fn SetInstrumentByChoice
  (params (InstrumentSelected :type EPilgrimTrackType%20Enum :default "TrackGuitar"))
  (CallFunction|ServerSetInstrumentbyChoice InstrumentSelected))

(fn SetSpecificDifficultyMode
  (params (PilgrimCorePlayerStateComponent :type Pilgrim%20Core%20Player%20State%20Component%20Object%20Reference) (DifficultyToChange :type EPilgrimSongDifficulty%20Enum :default "DifficultyEasy"))
  (bind _pilgrimplayspace (Variables|Default|GetPilgrimPlayspace))
  (bind _returnvalue (Class|PilgrimPlayspace|GetSupportedGameplayConfig _pilgrimplayspace))
  (bind _supporteddifficulties (Variables|/Script/PilgrimCoreRuntime.PilgrimSupportedGameplayConfig|GetSupportedDifficulties _returnvalue))
  (if (Utilities|Array|IsValidIndex _supporteddifficulties (Utilities|Array|FindItem _supporteddifficulties DifficultyToChange))
    (Class|PilgrimCorePlayerStateComponent|SetSongDifficulty PilgrimCorePlayerStateComponent DifficultyToChange)
    (elif (Utilities|Array|IsValidIndex _supporteddifficulties)
      (Class|PilgrimCorePlayerStateComponent|SetSongDifficulty PilgrimCorePlayerStateComponent (Utilities|Array|Get(aref) _supporteddifficulties))
      (else
        (Development|PrintString "ERROR: Supported Difficulties Array is empty. Default will be Easy." true true "(R=1.000000,G=0.028928,B=0.000000,A=1.000000)" 2e+02)))))

(fn SetupInstrumentFiltersFromPlaylistGameplayTags
  (params (InGameplayTags :type Gameplay%20Tag%20Container%20Structure))
  (if (GameplayTags|HasTag InGameplayTags "(TagName=\"Sparks.Playlist.Pilgrim.BattleStage.ScoreChase.Bass\")" true)
    (Variables|InstrumentFilters|SetbAllowBassSelect true)
    (Variables|InstrumentFilters|SetbAllowGuitarSelect false)
    (Variables|InstrumentFilters|SetbAllowVocalsSelect false)
    (Variables|InstrumentFilters|SetbAllowDrumsSelect false)
    (Variables|Default|SetForceWeaponSwap 1)
    (elif (GameplayTags|HasTag InGameplayTags "(TagName=\"Sparks.Playlist.Pilgrim.BattleStage.ScoreChase.Drums\")" true)
      (Variables|InstrumentFilters|SetbAllowBassSelect false)
      (Variables|InstrumentFilters|SetbAllowGuitarSelect false)
      (Variables|InstrumentFilters|SetbAllowVocalsSelect false)
      (Variables|InstrumentFilters|SetbAllowDrumsSelect true)
      (Variables|Default|SetForceWeaponSwap 3)
      (elif (GameplayTags|HasTag InGameplayTags "(TagName=\"Sparks.Playlist.Pilgrim.BattleStage.ScoreChase.Lead\")" true)
        (Variables|InstrumentFilters|SetbAllowBassSelect false)
        (Variables|InstrumentFilters|SetbAllowGuitarSelect true)
        (Variables|InstrumentFilters|SetbAllowVocalsSelect false)
        (Variables|InstrumentFilters|SetbAllowDrumsSelect false)
        (Variables|Default|SetForceWeaponSwap 2)
        (elif (GameplayTags|HasTag InGameplayTags "(TagName=\"Sparks.Playlist.Pilgrim.BattleStage.ScoreChase.Vocals\")" true)
          (Variables|InstrumentFilters|SetbAllowBassSelect false)
          (Variables|InstrumentFilters|SetbAllowGuitarSelect false)
          (Variables|InstrumentFilters|SetbAllowVocalsSelect true)
          (Variables|InstrumentFilters|SetbAllowDrumsSelect false)
          (Variables|Default|SetForceWeaponSwap 0))))))

(fn SwitchInstrumentOnPeripheralChange
  (params (IsTouchInputPlatform :type Boolean))
  (if (not IsTouchInputPlatform)
    (bind _asfort_player_controller (Utilities|Casting|CastToFortPlayerController (Components|GetOwner)))
    (Utilities|IsValid _asfort_player_controller
      (:Is%20Valid
        (bind _returnvalue (Class|PilgrimCorePlayerStateComponent|GetPilgrimCorePlayerStateComponent (Variables|/Script/Engine.Controller|GetPlayerState _asfort_player_controller)))
        (Utilities|IsValid _returnvalue
          (:Is%20Valid
            (CallFunction|ServerSetInstrumentandDifficulty "TrackGuitar" (Class|PilgrimCorePlayerStateComponent|GetSongDifficulty _returnvalue))))))))

(fn SwitchInstrumentOnPlatformChange
  (params (IsTouchInputPlatform :type Boolean))
  (bind _asfort_player_controller (Utilities|Casting|CastToFortPlayerController (Components|GetOwner)))
  (bind _returnvalue (Class|PilgrimCorePlayerStateComponent|GetPilgrimCorePlayerStateComponent (Variables|/Script/Engine.Controller|GetPlayerState _asfort_player_controller)))
  (bind _output_get (LocalVariables|SetPilgrimCorePlayerStateComponent _returnvalue))
  (LocalVariables|SetCurrent%20Instrument (Class|PilgrimCorePlayerStateComponent|GetInstrumentChoice _output_get))
  (bind _output_get_2 (LocalVariables|SetSupportedTrackTypes (select (Variables|GetIsTouchInputPlatform) (Variables|/Script/PilgrimCoreRuntime.PilgrimSupportedGameplayConfig|GetTouchInputSupportedTrackTypes (Class|PilgrimPlayspace|GetSupportedGameplayConfig (Variables|Default|GetPilgrimPlayspace))) (Variables|/Script/PilgrimCoreRuntime.PilgrimSupportedGameplayConfig|GetSupportedTrackTypes (Class|PilgrimPlayspace|GetSupportedGameplayConfig (Variables|Default|GetPilgrimPlayspace))))))
  (if (>= (Utilities|Array|FindItem _output_get_2 (LocalVariables|GetCurrent%20Instrument)) 0)
    _
    (elif (> (Utilities|Array|Length (LocalVariables|GetSupportedTrackTypes)) 0)
      ;[1 1 1] Fallback to the first available instrument
      (CallFunction|ServerSetInstrumentandDifficulty (Utilities|Array|Get(aref) (LocalVariables|GetSupportedTrackTypes)) (Class|PilgrimCorePlayerStateComponent|GetSongDifficulty (LocalVariables|GetPilgrimCorePlayerStateComponent))))))

(fn TimeSinceLastCrowdResponseChange
  (return (- (Math|DateTime|ToUnixTimestampDouble (Math|DateTime|Now)) (Variables|Default|GetLastCrowdResponseChangeTime))))

(fn ToggleCalibrateModal
  (bind _returnvalue (GameplayEventRouter|GetEventRouter "/Script/FortniteGame.FortPlayspace" (Components|GetOwner)))
  (GameplayEventRouter|BroadcastEvent _returnvalue (Utilities|Struct|MakeFMEventCalibrationToggleModal) 0))

(fn ToggleGreenroomMappingContext
  (params (AddGreenroomMappingContext :type Boolean))
  (bind _returnvalue (CallFunction|GetGreenRoomMappingContextPriority))
  (Default|ToggleMappingContext AddGreenroomMappingContext "/PilgrimSettings/Player/Abilities/IMC_GreenRoomMappingContext.IMC_GreenRoomMappingContext" (CallFunction|GetGreenRoomMappingContextPriority))
  ;[0.5722 1 0.9768 0.5] If a primary player, always remove the quick emoting IMC when entering/leaving the greenroom
  ;
  ; If not a primary player, add or remove the quick emoting IMC
  (if (Game|Player|IsPrimaryPlayer (Fortnite|GetFortPlayerControllerfromActor (Components|GetOwner)))
    (Default|ToggleMappingContext false "/PilgrimCore/Player/Abilities/IMC_GreenRoomMappingContext_QuickEmoting.IMC_GreenRoomMappingContext_QuickEmoting" _returnvalue)
    (else
      (Default|ToggleMappingContext AddGreenroomMappingContext "/PilgrimCore/Player/Abilities/IMC_GreenRoomMappingContext_QuickEmoting.IMC_GreenRoomMappingContext_QuickEmoting" _returnvalue)))
  ;[0.8634 0.01349 1] DEBUG
  (Fortnite|FortShippingLog 0 (Utilities|String|Append "[" (Utilities|GetDisplayName (Variables|Default|GetOwningPlayerController)) "] " "GreenroomMappingContext :" (select (Variables|GetAddGreenroomMappingContext) " added" " removed")))
  ;[0.07633 0.15 0.07118 0.5] Universal Part Picker IMC toggle
  (if (CallFunction|IsUniversalPartPickerEnabled)
    (Default|ToggleMappingContext AddGreenroomMappingContext "/PilgrimCore/Player/Abilities/IMC_PartPickerMappingContext.IMC_PartPickerMappingContext" (CallFunction|GetGreenRoomMappingContextPriority))))

(fn ToggleMappingContext
  (params (AddMappingContext :type Boolean) (MappingContext :type Input%20Mapping%20Context%20Object%20Reference) (Priority :type Integer :default 1000))
  ;[0.625 1 1] Add or remove the provided MappingContext
  (bind _asfort_player_controller (Utilities|Casting|CastToFortPlayerController (Components|GetOwner)))
  (if AddMappingContext
    (Input|AddMappingContext (PlayerController|LocalPlayerSubsystems|GetEnhancedInputLocalPlayerSubsystem _asfort_player_controller) MappingContext Priority "(bIgnoreAllPressedKeysUntilRelease=True,bForceImmediately=False,bNotifyUserSettings=False)")
    (else
      (Input|RemoveMappingContext (PlayerController|LocalPlayerSubsystems|GetEnhancedInputLocalPlayerSubsystem _asfort_player_controller) MappingContext "(bIgnoreAllPressedKeysUntilRelease=True,bForceImmediately=False,bNotifyUserSettings=False)"))))

(fn TrySwitchToInstrument
  (params (Track%20Type :type EPilgrimTrackType%20Enum :default "TrackGuitar"))
  (bind _localsupportedtracktypes (LocalVariables|GetLocalSupportedTrackTypes))
  (Utilities|IsPlayerAllowedtoSelectInstruments
    (:True
      (bind _output_get (LocalVariables|SetLocalSupportedTrackTypes (Variables|/Script/PilgrimCoreRuntime.PilgrimSupportedGameplayConfig|GetSupportedTrackTypes (Class|PilgrimPlayspace|GetSupportedGameplayConfig (CallFunction|GetPilgrimPlayspace)))))
      (if (Utilities|Array|IsValidIndex _localsupportedtracktypes (Utilities|Array|FindItem _output_get Track%20Type))
        (CallFunction|ServerTrySetInstrument (Utilities|Array|Get(aref) _localsupportedtracktypes (Utilities|Array|FindItem _output_get Track%20Type)))
        (else
          (CallFunction|ServerTrySetInstrument "TrackVocals"))))))

(fn TryToSwapToInstrumentIndex
  (params (Index :type Integer))
  ;[0.15 0.15 0.15 0.5 13] Universal Part Picker temporary blocker
  (if (not (CallFunction|IsUniversalPartPickerEnabled))
    (Utilities|IsPlayerAllowedtoSelectInstruments
      (:True
        (bind _etracktype (Default|GetTrackTypesEnumfromIndex Index))
        (bind _output_get (LocalVariables|SetLocalSupportedTrackTypes (Variables|/Script/PilgrimCoreRuntime.PilgrimSupportedGameplayConfig|GetSupportedTrackTypes (Class|PilgrimPlayspace|GetSupportedGameplayConfig (CallFunction|GetPilgrimPlayspace)))))
        (Utilities|FlowControl|Sequence
          (:then_0
            (Utilities|Array|RemoveItem _output_get "TrackProCymbals"))
          (:then_1
            (if (not (CallFunction|IsChoppedCheeseEnabled))
              (Utilities|Array|RemoveItem _output_get "TrackPlasticDrum")))
          (:then_2
            (if (not (CallFunction|IsChickenParmEnabled))
              (Utilities|Array|RemoveItem _output_get "TrackProVocals")))
          (:then_3
            (if (Utilities|Array|IsValidIndex _output_get (Utilities|Array|FindItem _output_get _etracktype))
              (CallFunction|ServerTrySetInstrument (Utilities|Array|Get(aref) _output_get (Utilities|Array|FindItem _output_get _etracktype)))
              (else
                (CallFunction|ServerTrySetInstrument "TrackVocals")))))))))

(fn UpdateCrowdResponse
  (Variables|Default|SetLastCrowdResponseChangeTime (Math|DateTime|ToUnixTimestampDouble (Math|DateTime|Now)))
  ;[0.5117 1 0.9099] Control whether crowd response SFX plays or not
  (if (Variables|Default|GetIsCrowdResponseAudioEnabled)
    (switch int (Variables|Default|GetCurrentCrowdResponseHealthTier)
      (:0
        ;[1 0 0.7242] SFX - Crowd cheers on band health changes:  to Poor, up to Good, up to Great, down to Good
        (Audio|PlaySound2D "/SparksMaps/Audio/CrowdResponsiveSFX/Sparks_CrowdDnToPoor.Sparks_CrowdDnToPoor"))
      (:1
        (if (not (Variables|Default|GetIsAtGreat))
          (Audio|PlaySound2D "/SparksMaps/Audio/CrowdResponsiveSFX/Sparks_CrowdUpToGood.Sparks_CrowdUpToGood")
          (else
            (Audio|PlaySound2D "/SparksMaps/Audio/CrowdResponsiveSFX/Sparks_CrowdDnToGood.Sparks_CrowdDnToGood"))))
      (:2
        (Audio|PlaySound2D "/SparksMaps/Audio/CrowdResponsiveSFX/Sparks_CrowdUpToGreat.Sparks_CrowdUpToGreat")
        (Variables|Default|SetIsAtGreat true)))
    (Variables|Default|SetIsAtGreat false)))

(fn UpdateCrowdResponseHealthTiers
  (params (PilgrimEvent_Scoring_PilgrimHealthChanged :type Pilgrim%20Event%20Band%20State%20Band%20Health%20Changed%20Structure%20%28by%20ref%29))
  (bind (_lastbandhealth _newbandhealth _lastbandhealthtier _newbandhealthtier) (Utilities|Struct|BreakPilgrimEventBandStateBandHealthChanged PilgrimEvent_Scoring_PilgrimHealthChanged))
  (Variables|Default|SetLastCrowdResponseHealthTier _lastbandhealthtier)
  (Variables|Default|SetCurrentCrowdResponseHealthTier _newbandhealthtier))
