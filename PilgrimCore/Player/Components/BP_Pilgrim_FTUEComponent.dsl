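; Base player-controller component that drives first-time-user-experience modals for Pilgrim playlists, with child classes overriding seen-state and widget classes.
; Once the playlist is ready, gameplay tags select How To Play, Controls Preview and Ready To Play widgets.
; After calibration, a delay timer pushes them in a flow-container modal.
; Event-router listeners record skip or completion, then offer local-multiplayer (HamSandwich) confirmation to the primary player in Quickplay.
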(event EventBeginPlay
  ;[1 0.5402 0] NOTE: Playlists that wish to add this component to their player controllers to show FTUE widgets should create their own component that inherits from this class, and set the relevant widget classes for that playlist in the child. Also, they should override the HasUserSeenFTUE and SetUserHasSeenFTUE functions.
  (bind _is_component_enabled (CallFunction|IsFTUEComponentEnabled))
  (if _is_component_enabled
    ;[0.5117 1 0.9099] Wait for the playlist to be ready, and if they haven't seen the FTUE before, check to see if we are in a valid playlist, then check for relevant gameplay tags so that we know which modals to display
    (bind (_gamestate _playlist) (Class|FortAsyncActionCurrentPlaylistReady|PlaylistDataReadyAsync)
      (:OnReady
        (bind _asfort_player_controller (Utilities|Casting|CastToFortPlayerController (Components|GetOwner)))
        (Variables|Default|SetAs%20Fort%20Player%20Controller _asfort_player_controller)
        (bind _asfort_playlist_athena (Utilities|Casting|CastToFortPlaylistAthena _playlist))
        (if (not (Utilities|String|IsEmpty (Variables|Default|GetRequiredCVar)))
          (bind _returnvalue (Development|GetConsoleVariableBoolValue (Variables|Default|GetRequiredCVar)))
          (if (not _returnvalue)
            (Development|PrintString "The required CVar for this FTUEComponent is not enabled.")))
        (if (and (Utilities|GameplayTagContainerNotEmpty (Variables|Default|GetRequiredPlaylistTags)) (GameplayTags|HasAnyTags (Variables|/Script/FortniteGame.FortPlaylist|GetGameplayTagContainer _asfort_playlist_athena) (Variables|Default|GetRequiredPlaylistTags) true))
          (Variables|Settings|SetShouldShowHowToPlayScreen (GameplayTags|HasAnyTags (Variables|/Script/FortniteGame.FortPlaylist|GetGameplayTagContainer _asfort_playlist_athena) (GameplayTags|MakeLiteralGameplayTagContainer "(GameplayTags=((TagName=\"Mojo.FMCalibration.FTUE.ShowHowToPlay\"),(TagName=\"Mojo.FMCalibration.FTUE.ShowAllFTUE\")))") true))
          (Variables|Settings|SetShouldShowControlsPreviewScreen (and (GameplayTags|HasAnyTags (Variables|/Script/FortniteGame.FortPlaylist|GetGameplayTagContainer _asfort_playlist_athena) (GameplayTags|MakeLiteralGameplayTagContainer "(GameplayTags=((TagName=\"Mojo.FMCalibration.FTUE.ShowAllFTUE\"),(TagName=\"Mojo.FMCalibration.FTUE.ShowControlsPreview\")))") true) (not (or (Class|PilgrimCoreBPFL|ShouldUseTouchUIForLocalPlayer) (Class|PilgrimPlayerControllerComponent|GetIsCurrentDevicePeripheral (Actor|GetComponentbyClass (Fortnite|GetControllerfromActor (Components|GetOwner)) "/PilgrimCore/Player/Components/BP_PilgrimPlayerControllerComponent.BP_PilgrimPlayerControllerComponent_C"))))))
          (Variables|Settings|SetShouldShowReadyToPlayScreen (GameplayTags|HasAnyTags (Variables|/Script/FortniteGame.FortPlaylist|GetGameplayTagContainer _asfort_playlist_athena) (GameplayTags|MakeLiteralGameplayTagContainer "(GameplayTags=((TagName=\"Mojo.FMCalibration.FTUE.ShowAllFTUE\"),(TagName=\"Mojo.FMCalibration.FTUE.ShowReadyToPlay\")))") true))
          (CallFunction|InitializeWidgetstoShowArray)
          ;[0.5804 1 0.93] Get the event router to setup event listeners
          (bind _returnvalue_2 (GameplayEventRouter|GetEventRouter "/Script/FortniteGame.FortPlayspace" (Components|GetOwner)))
          (Utilities|FlowControl|Sequence
            (:then_0
              ;[0.5117 1 0.9099] Start listening for the widget modal to be closed, and show the FTUE modals if the player hasn't seen them yet
              (GameplayEventRouter|StartListeningToEvent _returnvalue_2 "/Script/FMCalibrationRuntime.FMEvent_Calibration_ModalClosed"
                (:OnEventReceived
                  (bind _hasseenftue (CallFunction|HasUserSeenFTUE))
                  (if _hasseenftue
                    (CallFunction|OnFTUEFlowComplete)
                    (else
                      (Utilities|FlowControl|DelayUntilNextTick)
                      (CallFunction|AttempttoShowFTUEModal))))))
            (:then_1
              ;[0.6593 0.9834 1] Listen for a modal skipped event and mark that the user has seen the FTUE so it isn't shown again
              (GameplayEventRouter|StartListeningToEvent _returnvalue_2 "/Script/FMCalibrationRuntime.FMEvent_Calibration_ModalSkipped"
                (:OnEventReceived
                  ;;[1 0.7186 0.09945] Save that the user has seen the FTUE
                  (CallFunction|SetUserHasSeenFTUE true)
                  ;[0.07819 0.491 0.2232 0.502] Logic on the left :D
                  (CallFunction|OpenHamSandwichConfirmationPopup))))
            (:then_2
              ;[0.6593 0.9834 1] Listen for a FTUE completed event and mark that the user has seen the FTUE so it isn't shown again
              (GameplayEventRouter|StartListeningToEvent _returnvalue_2 "/Script/PilgrimCoreUI.PilgrimEvent_FTUE_Complete"
                (:OnEventReceived
                  ;;[1 0.7186 0.09945] Save that the user has seen the FTUE
                  (CallFunction|SetUserHasSeenFTUE true)
                  ;[0.07819 0.491 0.2232 0.502] Logic on the left :D
                  (CallFunction|OpenHamSandwichConfirmationPopup))))
            (:then_3
              ;[0.5804 1 0.93] Used by child classes to listen for other events specific to their playlists
              (CallFunction|SetupAdditionalEventListeners _returnvalue_2)))
          (CallFunction|OnFTUEFlowComplete)
          (else
            (Development|PrintString "No matching gameplay tags found on the playlist when trying to show the FTUE. Make sure gameplay tags are set on this component that match the playlist.")))))
    (else
      (Development|PrintString "This FTUEComponent is currently disabled. Please see IsFTUEComponentEnabled for more details"))))

(event EventDisplayFTUEWidgets
  (bind _returnvalue (Components|GetOwner))
  (bind _returnvalue_2 (Actor|GetComponentbyClass _returnvalue "/Script/FMCalibrationRuntime.FMCalibrationControllerComponent"))
  ;[0.5415 0.9901 1] Called from code, if the user has calibrated AND they have not seen the FTUE, attempt to show the FTUE modal after a short delay
  (Utilities|IsValid _returnvalue_2
    (:Is%20Valid
      (bind _hasseenftue (CallFunction|HasUserSeenFTUE))
      (if (and (not _hasseenftue) (Class|FMCalibrationControllerComponent|GetIsCalibrated _returnvalue_2))
        (Utilities|Time|SetTimerbyEvent (delegate Custom|ShowFTUEModalsWhenTimerEnds) (ScalableFloat|GetValueatLevel (Variables|Default|GetFTUEDelay)))
        (else
          ;[0.07819 0.491 0.2232 0.502] Only be called when the player has seen FTUE, this logic should open the Hamsandwich popup confirmation and setup the event to be called when the setup is closed
          (Utilities|Time|SetTimerbyEvent (delegate Custom|OpenHamSandwichConfirmationPopup) (ScalableFloat|GetValueatLevel (Variables|Default|GetFTUEDelay))))))))

(event Custom|ShowFTUEModalsWhenTimerEnds
  ;[0.5415 0.9901 1] Called from code, if the user has calibrated AND they have not seen the FTUE, attempt to show the FTUE modal after a short delay
  (CallFunction|AttempttoShowFTUEModal))

(event Custom|OpenHamSandwichConfirmationPopup
  ;[0.07819 0.491 0.2232 0.502] Only be called when the player has seen FTUE, this logic should open the Hamsandwich popup confirmation and setup the event to be called when the setup is closed
  (bind (_gamestate _playlist) (Class|FortAsyncActionCurrentPlaylistReady|PlaylistDataReadyAsync)
    (:OnReady
      (bind _asfort_playlist_athena (Utilities|Casting|CastToFortPlaylistAthena _playlist))
      (bind _hasseenftue (CallFunction|HasUserSeenFTUE))
      (if (Math|Boolean|ANDBoolean _hasseenftue (GameplayTags|HasAnyTags (Variables|/Script/FortniteGame.FortPlaylist|GetGameplayTagContainer _asfort_playlist_athena) (GameplayTags|MakeLiteralGameplayTagContainer "(GameplayTags=((TagName=\"Sparks.Playlist.Pilgrim.Quickplay\")))") true) (Class|PilgrimCoreBPFL|IsHamSandwichSessionActive) (CallFunction|IsOwnerPrimaryPlayer))
        (if (Default|HasUserSeenFTUELocalMultiplayer)
          (Development|PrintString "Player has already seen the Local MP FTUE so we will not show it")
          (else
            (bind _asfort_player_controller (Utilities|Casting|CastToFortPlayerController (Components|GetOwner))
              (:then
                (CallFunction|SetHasSeenFTUELocalMultiplayer true)
                (bind _userwidget (Class|FortAsyncActionCreateWidgetAsync|CreateWidgetAsync "/FMCalibration/Widgets/WBP_FMCalibration_HamSandwichConfirmation.WBP_FMCalibration_HamSandwichConfirmation_C" _asfort_player_controller)
                  (:OnComplete
                    (UserInterface|Viewport|AddtoViewport _userwidget 0)
                    ; Redundant, attempting long shot
                    (CallFunction|SetHasSeenFTUELocalMultiplayer true))))
              (:CastFailed
                (Development|PrintString "Failed to cast to fortplayer controller" false)))))))))

(event Custom|SetupAdditionalEventListeners
  (params (EventRouter :type Gameplay%20Event%20Router%20Component%20Object%20Reference)))

(fn AddWidgetsToShowToArray
  (params (CanShowWidget :type Boolean) (NewWidgetToShow :type Array%20of%20User%20Widget%20Class%20References))
  ;[0.6775 1 0.9229] Check if the condition is met to show this group of widgets, then loop through all of them.
  (if CanShowWidget
    (for _array_element NewWidgetToShow
      ;[1 0.826 0.2757] Try to cast this widget to a PilgrimFTUEWidgetBase class ref. If we can't, add it by default
      ;
      ; Note: Ideally we don't need to support general widgets, but I'm leaving this open for now to make sure the Controls Scheme screen can be easily added to the FTUE when it comes in.
      (bind _aspilgrim_ftuewidget_base (Utilities|Casting|CastToPilgrimFTUEWidgetBaseClass _array_element)
        (:then
          ;[0.5804 1 0.93] If we successfully cast to a PilgrimFTUEWidgetBase, check to see if it has all of the required playlist tags. If it does, add this widget to the list of widgets to show.
          (if (GameplayTags|HasAllTags (Class|GetClassDefaults _aspilgrim_ftuewidget_base) (Variables|Default|GetRequiredPlaylistTags))
            (Utilities|Array|AddUnique (Variables|Widgets|GetWidgetsToShow) _aspilgrim_ftuewidget_base)))
        (:CastFailed
          (Utilities|Array|AddUnique (Variables|Widgets|GetWidgetsToShow) _array_element))))))

(fn AttemptToShowFTUEModal
  ;[0.611 1 0.9901] Check to see if there are any widgets to show
  (if (Utilities|Array|IsNotEmpty (Variables|Widgets|GetWidgetsToShow))
    ;[0.7618 1 0.9701] Create the base FTUE flow widget, and pass it the widgets to show based on playlist tags, then push it as a modal widget
    (bind _asfort_player_controller (Utilities|Casting|CastToFortPlayerController (Components|GetOwner)))
    (bind _returnvalue (UserInterface|CreateWidget "/PilgrimCore/Widgets/FTUE/WBP_PilgrimCoreFTUE_FlowContainer.WBP_PilgrimCoreFTUE_FlowContainer_C" _asfort_player_controller (Variables|Widgets|GetWidgetsToShow)))
    (FortUIManagement|TryPushWidgetasModalPanelforPlayer _asfort_player_controller _returnvalue)
    (Variables|Default|SetFTUEModalWidget _returnvalue)))

(fn CloseAllFTUEScreens
  (bind _ftuemodalwidget (Variables|Default|GetFTUEModalWidget))
  ;[0.5804 1 0.93] If we have a valid reference to the FTUE modal widget, deactivate it
  (Utilities|IsValid _ftuemodalwidget
    (:Is%20Valid
      (ActivatableWidget|DeactivateWidget _ftuemodalwidget))))

(fn HasUserSeenFTUE
  (Development|PrintString "GetUserHasSeenFTUE function has not been overridden" true true "(R=1.000000,G=0.195490,B=0.025443,A=1.000000)")
  (return false))

(fn HasUserSeenFTUELocalMultiplayer
  (bind _returnvalue (CallFunction|GetHasSeenFTUELocalMultiplayer))
  ;[1 0.4146 0] Child classes can override this.
  ; Temporary Debug
  (Development|PrintString (Utilities|String|Append "Retrieved Has Seen FTUELocal: " (Utilities|String|ToString(Boolean) _returnvalue)))
  (return _returnvalue))

(fn InitializeWidgetsToShowArray
  ;[0.642 1 0.9501] Setup an array of widget classes that will be added, dependent on the conditions set by the playlist tags
  (Utilities|FlowControl|Sequence
    (:then_0
      (CallFunction|AddWidgetstoShowtoArray (Variables|Settings|GetShouldShowHowToPlayScreen) (Variables|Widgets|GetHowToPlayWidgetClasses)))
    (:then_1
      (CallFunction|AddWidgetstoShowtoArray (Variables|Settings|GetShouldShowControlsPreviewScreen) (Variables|Widgets|GetControlsPreviewWidgetClasses)))
    (:then_2
      (CallFunction|AddWidgetstoShowtoArray (Variables|Settings|GetShouldShowReadyToPlayScreen) (Variables|Widgets|GetReadyToPlayWidgetClasses)))))

(fn IsFTUEComponentEnabled
  ;[1 0.4146 0] Child classes should override this!
  (return true))

(fn IsOwnerPrimaryPlayer
  (bind _returnvalue (Components|GetOwner))
  (bind _returnvalue_2 (Actor|GetComponentbyClass _returnvalue "/PilgrimCore/Player/Components/BP_PilgrimPlayerControllerComponent.BP_PilgrimPlayerControllerComponent_C"))
  ;[0.06104 0.06003 0.15 0.5] This could probably be useful to extract into a BPFL
  (Utilities|IsValid _returnvalue_2
    (:Is%20Valid
      (return (Class|PilgrimPlayerControllerComponent|IsPrimaryLocalPlayer _returnvalue_2)))
    (:Is%20Not%20Valid
      ;;[0.15 0.1222 0.0271 0.5] Assume it's not primary player if PlayerControllerComponent was invalid
      (return false))))

(fn SetUserHasSeenFTUE
  (params (UserHasSeenFTUE :type Boolean))
  (Development|PrintString "SetUserHasSeenFTUE function has not been overridden" true true "(R=1.000000,G=0.195490,B=0.025443,A=1.000000)"))
